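; Anim notify for a MotoWrestle slash impact effect that configures the effect spawned by the parent notify.
; After the parent call, it passes the UseGroundDecal and ColorVariant variables to the spawned effect as parameters.
; With ground decal enabled, it also spawns an FN_RadialForce physics field actor at the effect's location.
; That spawn is gated by the AllowJanus check and the r.PhysicsField.EnableField console variable.
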(fn Received_Notify
  (params (MeshComp :type Skeletal%20Mesh%20Component%20Object%20Reference) (Animation :type Anim%20Sequence%20Base%20Object%20Reference) (EventReference :type Anim%20Notify%20Event%20Reference%20Structure%20%28by%20ref%29))
  (bind _returnvalue (|CallParentFunction MeshComp Animation EventReference))
  (bind _returnvalue_2 (AnimNotify|GetSpawnedEffect))
  (Utilities|IsValid _returnvalue_2
    (:Is%20Valid
      (Effects|Components|ParticleSystem|SetBoolParameter _returnvalue_2 "UseGroundDecal" (Variables|Default|GetUseGroundDecal))
      (Effects|Components|ParticleSystem|SetIntParameter _returnvalue_2 "ColorVariant" (Variables|Default|GetColorVariant)) ; Use in case not triggered by AnimSequence on player
      (if (Variables|Default|GetUseGroundDecal)
        ;[1 1 1] Add RadialForce
        (Utilities|AllowJanus
          (:Allow
            (bind _returnvalue_3 (Development|GetConsoleVariableIntValue "r.PhysicsField.EnableField"))
            (if (== _returnvalue_3 1)
              (Game|SpawnActorfromClass "/Game/Effects/Niagara/Blueprints/PhsyicsFields/FN_RadialForce.FN_RadialForce_C" (Math|Conversions|ToTransform(Vector) (Transformation|GetWorldLocation _returnvalue_2)) "Undefined" "MultiplyWithRoot" 0 0.2 0.6 1e+03 1.0 0.0 0)
              (else
                _))
            (return _returnvalue))
          (:Nope
            _))
        (else
          _)))
    (:Is%20Not%20Valid
      _)))
