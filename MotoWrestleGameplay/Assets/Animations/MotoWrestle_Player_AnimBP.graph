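# Player Animation Blueprint for MotoWrestle that implements Item Override and LocomotionOverride layers, with most left unwired and no event-graph logic.
# While the lower-body corrective flag is set, full locomotion layers a non-targeted pose from Spine_01 plus an idle additive with slope-damped legs.
# Under that flag, a state machine blends equip and corrective additives into the lower body, then offsets pelvis height.
# In vehicles, the full-body layer twists pelvis yaw.
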AnimBlueprint /MotoWrestleGameplay/Assets/Animations/MotoWrestle_Player_AnimBP
target_skeleton=&Fortnite_M_Avg_Player_Skeleton
graphs [
 AnimGraph {nodes [Output Pose()]}
 EventGraph {links [], nodes []}
 ItemAimOffset {nodes [
   Input Pose!() -> (Pose, AimOffsetAlpha:double, Yaw:double, Pitch:double) {name=InPose_AimOffset}
   Input Pose!() -> (Pose) {name=InPose_PreAimOffsets}
   Output Pose() {layerGroup=Item Override}
  ]}
 ItemFallAdditive {nodes [
   Input Pose() {name=InPoseItemFallAdditive}
   Output Pose() {layerGroup=Item Override}
  ]}
 ItemFallLandAdditive {nodes [
   Input Pose() {name=InPoseItemFallLandAdditive}
   Output Pose() {layerGroup=Item Override}
  ]}
 ItemFallLevel2Override {nodes [
   Input Pose() {name=InPose_FallLevel2}
   Output Pose() {layerGroup=Item Override}
  ]}
 ItemFinalPoseOverride {nodes [
   Input Pose() {name=InFinalPose}
   Output Pose() {layerGroup=Item Override}
  ]}
 ItemFullBodyAimOffsets {nodes [
   Input Pose!() -> (Pose, FullBody_AimOffsetAlpha:double, Aim_Yaw:double, Aim_Pitch:double) {name=InPoseFullBodyAimOffsets}
   Input Pose!() -> (Pose) {name=InPosePreFullBodyAimOffsets}
   Output Pose() {layerGroup=Item Override}
  ]}
 ItemFullBodyOverride {nodes [
   n1 = "Transform (Modify) Bone - Bone: pelvis"!(
    ComponentPose=Local To Component(LocalPose=Use cached pose 'Item_Fullbody'())
    Alpha=1
    Translation="0,0,0"
    Rotation_Roll:float=0
    Rotation_Pitch:float=0
    Rotation_Yaw:float=Get CustomPelvisTwist() -> (CustomPelvisTwist:double)
    Scale="1,1,1"
   ) -> (Pose) {
    rotationMode=BMM_Additive
    translationMode=BMM_Additive
    alphaBoolBlend {blendOption=Linear}
    boneToModify {boneName=pelvis}
    translation {z=-0.6}
   }
   Output Pose(Result=Blend Poses by bool(
     BlendPose_0=Component To Local(ComponentPose=@n1)
     BlendPose_1=Use cached pose 'Item_Fullbody'()
     bActiveValue=Get bIsInVehicle() -> (bIsInVehicle:bool)
    ) {blendType=Linear}) {layerGroup=Item Override}
   Save cached pose 'Item_Fullbody'(Pose=Input Pose() {name=InPoseFullBody}) {cacheName=Item_Fullbody}
  ]}
 ItemIdleOverride {nodes [
   Input Pose() {name=InPose_Idle}
   Output Pose() {layerGroup=Item Override}
  ]}
 ItemJetPackJumpAdditive {nodes [Output Pose() {layerGroup=Item Override}]}
 ItemJetPackStartAdditive {nodes [Output Pose() {layerGroup=Item Override}]}
 ItemJumpLoopAdditive {nodes [
   Input Pose() {name=InPoseItemJumpLoopAdditive}
   Output Pose() {layerGroup=Item Override}
  ]}
 ItemJumpOverride {nodes [
   Input Pose() {name=InPoseJump}
   Output Pose() {layerGroup=Item Override}
  ]}
 ItemJumpUpAdditive {nodes [
   Input Pose() {name=InPoseItemJumpUpAdditive}
   Output Pose() {layerGroup=Item Override}
  ]}
 ItemPreAimOffsets {nodes [
   Input Pose() {name=InPosePreAimOffsets}
   Output Pose() {layerGroup=Item Override}
  ]}
 ItemPreIK {nodes [
   Input Pose() {name=InPose_PreIK}
   Output Pose() {layerGroup=Item Override}
  ]}
 ItemRightHandIK {nodes [
   Input Pose() {name=InPosePreRightHandIK}
   Input Pose() {name=InPosePostRightHandIK}
   Output Pose() {layerGroup=Item Override}
  ]}
 ItemSkeletalControl {nodes [
   Input Pose()
   Output Pose() {layerGroup=Item Override}
  ]}
 ItemSwimJumpFallLoop {nodes [Output Pose() {layerGroup=Item Override}]}
 ItemSwimJumpSurfaceEnd {nodes [Output Pose() {layerGroup=Item Override}]}
 ItemSwimJumpSurfaceLoop {nodes [Output Pose() {layerGroup=Item Override}]}
 ItemUpperBody {nodes [
   Input Pose!() -> (Pose, InputParam:Fort Anim Input Adjusted Aim Structure) {name=InPoseUpperBody}
   Output Pose() {layerGroup=Item Override}
  ]}
 ItemZipLineStartAdditive {nodes [Output Pose() {layerGroup=Item Override}]}
 LocomotionLayer_Additives {nodes [Output Pose() {layerGroup=LocomotionOverride}]}
 LocomotionLayer_BaseLocomotion {nodes [
   Input Pose!() -> (Pose) {name=InPose_BaseLocomotion}
   Output Pose() {layerGroup=LocomotionOverride}
  ]}
 LocomotionLayer_FinalPose {nodes [
   Input Pose() {name=InPoseFinalPose}
   Output Pose() {layerGroup=LocomotionOverride}
  ]}
 LocomotionLayer_FullBody {nodes [
   Input Pose() {name=InPose_FullBody}
   Output Pose() {layerGroup=LocomotionOverride}
  ]}
 LocomotionLayer_FullLocomotion {nodes [
   Output Pose(Result=Blend Poses by bool(
     BlendPose_0=Use cached pose 'UpperBodyIdlePose'()
     BlendPose_1=Use cached pose 'InPose_FullBody'()
     BlendTime_0=0.4
     BlendTime_1=0.3
     bActiveValue=Get bApplyLowerBodyCorrective() -> (bApplyLowerBodyCorrective:bool)
    ) {blendType=CubicInOut}) {layerGroup=LocomotionOverride}
   Save cached pose 'InPose_FullBody'(Pose=Input Pose() {name=InPose_FullLocomotion}) {cacheName=InPose_FullBody}
   n1 = Sequence Player 'MotoWrestle_NonTargeted_Pose'() {bLoopAnimation=false, sequence=&/MotoWrestle_NonTargeted_Pose}
   n2 = Use cached pose 'InPose_FullBody'()
   n3 = Get bBlendUpperBodySourcePoseVal() -> (bBlendUpperBodySourcePoseVal:bool)
   Save cached pose 'UpperBodySourcePoseOverride'(Pose=Layered blend per bone(
     BasePose=Use cached pose 'InPose_FullBody'()
     BlendPoses_0=Blend Poses by bool(BlendPose_0=@n1, BlendPose_1=@n2, BlendTime_0=1, BlendTime_1=1, bActiveValue=@n3) {blendType=CubicInOut}
    ) {bMeshSpaceRotationBlend=true, layerSetup [{branchFilters [{boneName=Spine_01}]}]}) {cacheName=UpperBodySourcePoseOverride}
   n4 = Layered blend per bone(
    BasePose="Sequence Player (additive) 'MotoWrestle_Idle_Loop'"() {sequence=&/MotoWrestle_Idle_Loop}
    BlendPoses_0=Additive Identity Pose()
    BlendWeights_0=Get LowerBodyAdditiveLoopSlopeDampFactor() -> (LowerBodyAdditiveLoopSlopeDampFactor:double)
   ) {
    bMeshSpaceRotationBlend=true
    blendMode=BlendMask
    blendMasks [&Fortnite_M_Avg_Player_Skeleton:LowerBody]
    layerSetup []
   }
   Save cached pose 'UpperBodyIdlePose'(Pose=Apply Additive(Base=Use cached pose 'UpperBodySourcePoseOverride'(), Additive=@n4, Alpha=Get IdleAdditiveAlpha() -> (IdleAdditiveAlpha:double)) {alphaBoolBlend {blendOption=Linear}}) {cacheName=UpperBodyIdlePose}
  ]}
 LocomotionLayer_GroundLandAdditive {nodes [
   Input Pose()
   Output Pose() {layerGroup=LocomotionOverride}
  ]}
 LocomotionLayer_InAir {nodes [
   Input Pose() {name=InPose_InAir}
   Input Pose() {name=InPose_InAirSourcePose}
   Output Pose() {layerGroup=LocomotionOverride}
  ]}
 LocomotionLayer_LowerBody {nodes [
   AnimGraphNode_StateMachine_0 = CorrectivePoseBlend() {is=title AnimGraphNode_StateMachine, bAllowConduitEntryStates=true, maxTransitionsPerFrame=1}
   n1 = "Transform (Modify) Bone - Bone: pelvis"!(
    ComponentPose=Local To Component(LocalPose=@AnimGraphNode_StateMachine_0)
    Alpha=1
    Translation_X:double=0
    Translation_Y:double=0
    Translation_Z:double=Get HipOffsetVertical() -> (HipOffsetVertical:double)
    Rotation:Rotator="0,0,0"
    Scale="1,1,1"
   ) -> (Pose) {translationMode=BMM_Additive, alphaBoolBlend {blendOption=Linear}, boneToModify {boneName=pelvis}, translation {z=-0.6}}
   Output Pose(Result=Blend Poses by bool(
     BlendPose_0=Component To Local(ComponentPose=@n1)
     BlendPose_1=Use cached pose 'InPose_LowerBody'()
     BlendTime_0=0.5
     BlendTime_1=0.4
     bActiveValue=Get bApplyLowerBodyCorrective()
    ) {blendType=CubicInOut}) {layerGroup=LocomotionOverride}
   Save cached pose 'InPose_LowerBody'(Pose=Input Pose() {name=InPose_LowerBody}) {cacheName=InPose_LowerBody}
  ]}
 CorrectivePoseBlend {
  entry=PassThrough
  path=LocomotionLayer_LowerBody.AnimGraphNode_StateMachine_0.CorrectivePoseBlend
  ids=id
  nodes [
   AnimStateNode_0 = PassThrough() {
    is=title AnimStateNode
    stateEntered {guid=#1}
    stateFullyBlended {guid=#2}
    stateLeft {guid=#3}
   }
   AnimStateNode_1 = EquipAdditive() {
    is=title AnimStateNode
    stateEntered {guid=#4}
    stateFullyBlended {guid=#5}
    stateLeft {guid=#6}
   }
   AnimStateNode_2 = CorrectiveAdditive() {
    is=title AnimStateNode
    stateEntered {guid=#7}
    stateFullyBlended {guid=#8}
    stateLeft {guid=#9}
   }
   AnimStateTransitionNode_0 = AnimStateTransitionNode() {blendMode=CircularOut, crossfadeDuration=0.6}
   AnimStateTransitionNode_1 = AnimStateTransitionNode() {crossfadeDuration=0.3}
   AnimStateTransitionNode_2 = AnimStateTransitionNode() {crossfadeDuration=0.45}
   AnimStateTransitionNode_3 = AnimStateTransitionNode() {
    transitionEnd {guid=#10}
    transitionInterrupt {guid=#11}
    transitionStart {guid=#12}
   }
   AnimStateTransitionNode_4 = AnimStateTransitionNode()
  ]
  links [
   {
    dst=PassThrough
    rule=LocomotionLayer_LowerBody.AnimGraphNode_StateMachine_0.CorrectivePoseBlend.AnimStateTransitionNode_2.Transition
    src=CorrectiveAdditive
    via=AnimStateTransitionNode_2
   }
   {
    dst=CorrectiveAdditive
    rule=LocomotionLayer_LowerBody.AnimGraphNode_StateMachine_0.CorrectivePoseBlend.AnimStateTransitionNode_3.Transition
    src=EquipAdditive
    via=AnimStateTransitionNode_3
   }
   {
    dst=PassThrough
    rule=LocomotionLayer_LowerBody.AnimGraphNode_StateMachine_0.CorrectivePoseBlend.AnimStateTransitionNode_4.Transition
    src=EquipAdditive
    via=AnimStateTransitionNode_4
   }
   {
    dst=CorrectiveAdditive
    rule=LocomotionLayer_LowerBody.AnimGraphNode_StateMachine_0.CorrectivePoseBlend.AnimStateTransitionNode_0.Transition
    src=PassThrough
    via=AnimStateTransitionNode_0
   }
   {
    dst=EquipAdditive
    rule=LocomotionLayer_LowerBody.AnimGraphNode_StateMachine_0.CorrectivePoseBlend.AnimStateTransitionNode_1.Transition
    src=PassThrough
    via=AnimStateTransitionNode_1
   }
  ]
 }
 PassThrough {
  path=LocomotionLayer_LowerBody.AnimGraphNode_StateMachine_0.CorrectivePoseBlend.AnimStateNode_0.PassThrough
  nodes [Output Animation Pose(Result=Use cached pose 'InPose_LowerBody'())]
 }
 EquipAdditive {
  path=LocomotionLayer_LowerBody.AnimGraphNode_StateMachine_0.CorrectivePoseBlend.AnimStateNode_1.EquipAdditive
  nodes [
   n1 = "Sequence Player (additive) 'MotoWrestle_equip'"() {bLoopAnimation=false, sequence=&/MotoWrestle_equip}
   n2 = Use cached pose 'InPose_LowerBody'()
   Output Animation Pose(Result=Blend Poses by bool(
     BlendPose_0=Use cached pose 'InPose_LowerBody'()
     BlendPose_1=Apply Additive(Base=@n2, Additive=@n1) {alphaBoolBlend {blendOption=Linear}}
     BlendTime_0=0.2
     BlendTime_1=0.2
     bActiveValue=Get bIsMoving2D()
    ) {blendType=Linear})
  ]
 }
 CorrectiveAdditive {
  path=LocomotionLayer_LowerBody.AnimGraphNode_StateMachine_0.CorrectivePoseBlend.AnimStateNode_2.CorrectiveAdditive
  nodes [Output Animation Pose(Result=Apply Additive(
     Base=Use cached pose 'InPose_LowerBody'()
     Additive="Sequence Player (additive) 'PoppySoap_NonTargeted_Pose_Add'"() {bLoopAnimation=false, sequence=/PoppySoap/Assets/Animation/PoppySoap_NonTargeted_Pose_Add}
    ) {alphaBoolBlend {blendOption=Linear}})]
 }
 Transition {
  path=LocomotionLayer_LowerBody.AnimGraphNode_StateMachine_0.CorrectivePoseBlend.AnimStateTransitionNode_0.Transition
  nodes [Result(bCanEnterTransition=Get bApplyLowerBodyCorrective())]
 }
 Transition {
  path=LocomotionLayer_LowerBody.AnimGraphNode_StateMachine_0.CorrectivePoseBlend.AnimStateTransitionNode_1.Transition
  nodes [Result(bCanEnterTransition=Get bIsBeingEquipped() -> (bIsBeingEquipped:bool))]
 }
 Transition {
  path=LocomotionLayer_LowerBody.AnimGraphNode_StateMachine_0.CorrectivePoseBlend.AnimStateTransitionNode_2.Transition
  nodes [Result(bCanEnterTransition=NOT Boolean(A=Get bApplyLowerBodyCorrective()))]
 }
 Transition {
  path=LocomotionLayer_LowerBody.AnimGraphNode_StateMachine_0.CorrectivePoseBlend.AnimStateTransitionNode_3.Transition
  nodes [Result(bCanEnterTransition=Get bEquipAdditiveToCorrectiveAdditiveVal() -> (bEquipAdditiveToCorrectiveAdditiveVal:bool))]
 }
 Transition {
  path=LocomotionLayer_LowerBody.AnimGraphNode_StateMachine_0.CorrectivePoseBlend.AnimStateTransitionNode_4.Transition
  nodes [Result(bCanEnterTransition=Get bEquipAdditiveToPassThroughVal() -> (bEquipAdditiveToPassThroughVal:bool))]
 }
 LocomotionLayer_UpperBody {nodes [
   Input Pose() {name=InPose_UpperBody}
   Output Pose() {layerGroup=LocomotionOverride}
  ]}
]
variables []
