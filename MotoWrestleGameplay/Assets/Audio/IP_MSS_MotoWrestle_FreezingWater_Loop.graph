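# MotoWrestle looping mono MetaSound for movement through freezing water, driven by the Fort.Movement speed inputs.
# Normalized speed feeds two crossfade segment managers that start and stop two randomly offset players of the same fast water loop.
# One layer's gain follows crossfade volume, the other's an ADSR envelope.
# An envelope follower on inverted absolute speed elevation scales the mix, reducing it during vertical movement.
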MetaSoundSource /MotoWrestleGameplay/Assets/Audio/IP_MSS_MotoWrestle_FreezingWater_Loop
inputs [
 In {type=Float}
 In 1 {type=Float}
 MinCrossfadeValue {type=Float}
 Movement.Speed {type=Float}
 Movement.SpeedAzimuth {type=Float}
 Movement.SpeedElevation {type=Float}
 Start Time {type=Time}
 UE.Source.OnPlay {type=Trigger}
]
interfaces [Fort.Movement:1.0, UE.OutputFormat.Mono:1.0, UE.Source:1.1]
outputs [
 L Mono {type=Audio}
 R Mono {type=Audio}
 "UE.OutputFormat.Mono.Audio:0" {type=Audio}
]
pages [Default {
  input_defaults {
   In=0
   In 1=0.5
   MinCrossfadeValue=0.65
   Movement.Speed=1000
   Movement.SpeedAzimuth=0
   Movement.SpeedElevation=0
   Start Time=0
   UE.Source.OnPlay=null
  }
  nodes [
   n1 = MapRange.MapRange:Float(In=^Movement.Speed) {also {defaults {In Range B=1100}}}
   n2 = Fortnite.CrossfadeSegmentManager:Float(In=@n1.Out Value) {also {defaults {Max=2.5, MaxCurve=0.5, MaxRange=0, Min=0.65, MinRange=0.7}}}
   WavePlayer_0 = UE.Wave Player:Mono(
    Play=@n2.Activate
    Stop=@n2.Deactivate
    Start Time=&MSP_WaveGetRandomStartTime() {also {defaults {Sound=&/IP_MotoWrestle_Water_Navigation_Loop_Fast}}}.Value
   ) {also {defaults {Loop="<MetaSoundBuilderResult.SUCCEEDED: 0>", Wave Asset=&IP_MotoWrestle_Water_Navigation_Loop_Fast}}}
   ADSREnvelopeADSREnvelope_0 = ADSR Envelope.ADSR Envelope:Float(Trigger Release=@n2.Deactivate, Trigger Attack=@WavePlayer_0.On Play) {also {defaults {Attack Curve=0.5, Attack Time=1, Decay Time=0.1, Hard Reset="<MetaSoundBuilderResult.SUCCEEDED: 0>", Release Curve=0.5, Sustain Level=1}}}
   n3 = Fortnite.CrossfadeSegmentManager:Float(In=UE.Add:Float(PrimaryOperand=@n1.Out Value) {also {defaults {AdditionalOperands=0.2}}}.Out) {also {defaults {Max=0.9, MaxCurve=2, MaxRange=0.2, MinCurve=0.7, MinRange=0.2}}}
   WavePlayer_1 = UE.Wave Player:Mono(
    Play=@n3.Activate
    Stop=@n3.Deactivate
    Start Time=&MSP_WaveGetRandomStartTime() {also {defaults {Sound=&IP_MotoWrestle_Water_Navigation_Loop_Fast}}}.Value
   ) {also {defaults {Loop="<MetaSoundBuilderResult.SUCCEEDED: 0>", Wave Asset=&IP_MotoWrestle_Water_Navigation_Loop_Fast}}}
   AudioMixerAudioMixerMono3_0 = "AudioMixer.Audio Mixer (Mono, 3)"(
    Gain 1=UE.Multiply:Float(PrimaryOperand=@ADSREnvelopeADSREnvelope_0.Out Envelope) {also {defaults {AdditionalOperands=1.5}}}.Out
    Gain 0=UE.Multiply:Float(PrimaryOperand=@n3.Volume) {also {defaults {AdditionalOperands=0.7}}}.Out
    In 1=@WavePlayer_0.Out Mono
    In 0=@WavePlayer_1.Out Mono
   )
   n4 = MapRange.MapRange:Float(In=AbsoluteValue.Abs:Float(Input=^Movement.SpeedElevation).Value)
   EnvelopeFollower_0 = UE.Envelope Follower(In=UE.ConversionFloatToAudio(In=UE.Subtract:Float(AdditionalOperands=@n4.Out Value) {also {defaults {PrimaryOperand=1}}}.Out).Out)
   Multiply_0 = UE.Multiply:Audio by Float(PrimaryOperand=@AudioMixerAudioMixerMono3_0.Out, AdditionalOperands=@EnvelopeFollower_0.Envelope)
  ]
  links [{dst="@UE.OutputFormat.Mono.Audio:0", src=Multiply_0.Out}]
 }]
properties {
 assetImportData=&self:AssetImportData
 attenuationSettings=/Game/Sounds/Fort_Player/Footsteps/Settings/Fort_Player_FS_3P_Att
 bLooping=true
 duration=10000
 maxDistance=3584
 sampleRate=48000
 soundClassObject=&Fort_Player_Foley
 volume=0.9
}
