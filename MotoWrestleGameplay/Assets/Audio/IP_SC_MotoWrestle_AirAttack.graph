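# Sound cue for the MotoWrestle air attack, playing one of four equally weighted kick whoosh waves from a shared random node.
# Branches on islocal and IsEnemy pick the mix: local, enemy and unset-enemy cases play at full modulator volume, while non-local non-enemies play at quarter volume.
# Non-local paths use close-range melee attenuation nodes.
# The cue uses the melee weapon sound class and overrides concurrency to three voices.
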SoundCue /MotoWrestleGameplay/Assets/Audio/IP_SC_MotoWrestle_AirAttack
root=Branch_0
nodes [
 {id=Attenuation_0, children [Modulator_2]}
 {
  id=Attenuation_1
  children [Modulator_3]
  properties {attenuationSettings=/Game/Sounds/Attenuation/Fort_Weapon_Melee_Attack_Close_Range_Stereo_Att}
 }
 {
  id=Attenuation_2
  children [Modulator_4]
  properties {attenuationSettings=&Fort_Weapon_Melee_Attack_Close_Range_Stereo_Att}
 }
 {id=Branch_0, children [Attenuation_0, Branch_3, Branch_3], properties {boolParameterName=islocal}}
 {id=Branch_3, children [Attenuation_2, Attenuation_1, Attenuation_2], properties {boolParameterName=IsEnemy}}
 {id=Modulator_2, children [Random_0], properties {pitchMax=1, pitchMin=1, volumeMax=1, volumeMin=1}}
 {id=Modulator_3, children [Random_0], properties {pitchMax=1, pitchMin=1, volumeMax=0.25, volumeMin=0.25}}
 {id=Modulator_4, children [Random_0], properties {pitchMax=1, pitchMin=1, volumeMax=1, volumeMin=1}}
 {
  id=Random_0
  children [WavePlayer_0,WavePlayer_1,WavePlayer_2,WavePlayer_3]
  properties {bShouldExcludeFromBranchCulling=true, weights [1,1,1,1]}
 }
 {id=WavePlayer_0, properties {soundWaveAssetPtr=&/IP_MotoWrestle_AirAttack_Kick_Whoosh_01}}
 {id=WavePlayer_1, properties {soundWaveAssetPtr=&/IP_MotoWrestle_AirAttack_Kick_Whoosh_02}}
 {id=WavePlayer_2, properties {soundWaveAssetPtr=&/IP_MotoWrestle_AirAttack_Kick_Whoosh_03}}
 {id=WavePlayer_3, properties {soundWaveAssetPtr=&/IP_MotoWrestle_AirAttack_Kick_Whoosh_04}}
]
properties {
 attenuationSettings=&Fort_Weapon_Melee_Attack_Close_Range_Stereo_Att
 bOverrideConcurrency=true
 maxDistance=2097152
 soundClassObject=&Fort_WeaponMelee
 volumeMultiplier=0.85
 concurrencyOverrides {maxCount=3}
}
