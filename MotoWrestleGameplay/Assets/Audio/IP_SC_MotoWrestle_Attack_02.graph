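# MotoWrestle attack sound cue that plays one of four equally weighted kick waves as a melee weapon sound.
# Branches on islocal, then IsEnemy, set attenuation and gain, with non-local non-enemy attacks cut to quarter volume and non-local ones using close-range melee attenuation.
# All paths share one random pool behind a modulator adding slight variation.
# A concurrency override limits simultaneous plays to three.
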SoundCue /MotoWrestleGameplay/Assets/Audio/IP_SC_MotoWrestle_Attack_02
root=Branch_1
nodes [
 {id=Attenuation_0, children [Modulator_2]}
 {
  id=Attenuation_1
  children [Modulator_3]
  properties {attenuationSettings=/Game/Sounds/Attenuation/Fort_Weapon_Melee_Attack_Close_Range_Mono_Att}
 }
 {
  id=Attenuation_2
  children [Modulator_4]
  properties {attenuationSettings=&Fort_Weapon_Melee_Attack_Close_Range_Mono_Att}
 }
 {id=Branch_1, children [Attenuation_0, Branch_3, Attenuation_0], properties {boolParameterName=islocal}}
 {id=Branch_3, children [Attenuation_2, Attenuation_1, Attenuation_2], properties {boolParameterName=IsEnemy}}
 {id=Modulator_0, children [Random_2], properties {volumeMax=1}}
 {id=Modulator_2, children [Modulator_0], properties {pitchMax=1, pitchMin=1, volumeMax=1, volumeMin=1}}
 {id=Modulator_3, children [Modulator_0], properties {pitchMax=1, pitchMin=1, volumeMax=0.25, volumeMin=0.25}}
 {id=Modulator_4, children [Modulator_0], properties {pitchMax=1, pitchMin=1, volumeMax=1, volumeMin=1}}
 {
  id=Random_2
  children [WavePlayer_4, WavePlayer_11, WavePlayer_12, WavePlayer_13]
  properties {bShouldExcludeFromBranchCulling=true, weights [1,1,1,1]}
 }
 {id=WavePlayer_11, properties {soundWaveAssetPtr=&/IP_MotoWrestle_Attack_Kick_02}}
 {id=WavePlayer_12, properties {soundWaveAssetPtr=&/IP_MotoWrestle_Attack_Kick_03}}
 {id=WavePlayer_13, properties {soundWaveAssetPtr=&/IP_MotoWrestle_Attack_Kick_04}}
 {id=WavePlayer_4, properties {soundWaveAssetPtr=&/IP_MotoWrestle_Attack_Kick_01}}
]
properties {
 bDebug=true
 bOverrideConcurrency=true
 maxDistance=2097152
 soundClassObject=&Fort_WeaponMelee
 volumeMultiplier=0.85
 attenuationOverrides {distanceAlgorithm=Logarithmic}
 concurrencyOverrides {maxCount=3}
}
