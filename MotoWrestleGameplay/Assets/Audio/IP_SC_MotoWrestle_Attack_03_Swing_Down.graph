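# Sound cue for the downward swing of MotoWrestle's third attack, in the melee weapon sound class with playback capped at three instances.
# A single wave with varied pitch and volume feeds a mixer shared by three listener paths.
# Branches on islocal and IsEnemy play local or unset at full level, non-local enemies slightly lower, and non-enemies much quieter.
# Both non-local paths use close-range melee attenuation.
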SoundCue /MotoWrestleGameplay/Assets/Audio/IP_SC_MotoWrestle_Attack_03_Swing_Down
root=Branch_0
nodes [
 {id=Attenuation_0, children [Modulator_2]}
 {
  id=Attenuation_1
  children [Modulator_3]
  properties {attenuationSettings=/Game/Sounds/Attenuation/Fort_Weapon_Melee_Attack_Close_Range_Mono_Att}
 }
 {
  id=Attenuation_2
  children [Modulator_4]
  properties {attenuationSettings=&Fort_Weapon_Melee_Attack_Close_Range_Mono_Att}
 }
 {id=Branch_0, children [Attenuation_0, Branch_3, Attenuation_0], properties {boolParameterName=islocal}}
 {id=Branch_3, children [Attenuation_2, Attenuation_1, Attenuation_2], properties {boolParameterName=IsEnemy}}
 {id=Mixer_0, children [null,Modulator_1], properties {inputVolume [1,1]}}
 {id=Modulator_1, children [WavePlayer_1], properties {pitchMax=1.2, volumeMax=1}}
 {id=Modulator_2, children [Mixer_0], properties {pitchMax=1, pitchMin=1, volumeMax=1, volumeMin=1}}
 {id=Modulator_3, children [Mixer_0], properties {pitchMax=1, pitchMin=1, volumeMax=0.25, volumeMin=0.25}}
 {id=Modulator_4, children [Mixer_0], properties {pitchMax=1, pitchMin=1, volumeMax=0.8, volumeMin=0.8}}
 {id=WavePlayer_1, properties {soundWaveAssetPtr=&/IP_MotoWrestle_Attack_03_Swing_Down_01}}
]
properties {
 bDebug=true
 bOverrideConcurrency=true
 maxDistance=2097152
 soundClassObject=&Fort_WeaponMelee
 volumeMultiplier=0.6
 attenuationOverrides {distanceAlgorithm=Logarithmic}
 concurrencyOverrides {maxCount=3}
}
