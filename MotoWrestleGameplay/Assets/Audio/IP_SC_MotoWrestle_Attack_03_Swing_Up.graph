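# Sound cue for the upswing of MotoWrestle's third attack, a hammer melee move in the melee weapon sound class, capped at three concurrent instances.
# Branches on islocal, then IsEnemy: local or unset plays swing, quieter whoosh and hammer-form layers.
# Remote players get only swing and form through close-range melee attenuation, with non-enemies much quieter.
# Each layer picks a random wave with modulated pitch.
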SoundCue /MotoWrestleGameplay/Assets/Audio/IP_SC_MotoWrestle_Attack_03_Swing_Up
root=Branch_0
nodes [
 {id=Attenuation_0, children [Modulator_2]}
 {
  id=Attenuation_1
  children [Modulator_3]
  properties {attenuationSettings=/Game/Sounds/Attenuation/Fort_Weapon_Melee_Attack_Close_Range_Mono_Att}
 }
 {
  id=Attenuation_2
  children [Modulator_4]
  properties {attenuationSettings=&Fort_Weapon_Melee_Attack_Close_Range_Mono_Att}
 }
 {id=Branch_0, children [Attenuation_0, Branch_3, Attenuation_0], properties {boolParameterName=islocal}}
 {id=Branch_3, children [Attenuation_2, Attenuation_1, Attenuation_2], properties {boolParameterName=IsEnemy}}
 {id=Mixer_0, children [Modulator_8,Modulator_6], properties {inputVolume [1,0.8]}}
 {id=Mixer_1, children [Modulator_5,Modulator_1,Modulator_0], properties {inputVolume [1,0.4,1]}}
 {id=Modulator_0, children [Random_0], properties {pitchMax=1.2, pitchMin=0.9, volumeMax=1}}
 {id=Modulator_1, children [Random_3], properties {pitchMax=1.2, pitchMin=0.9, volumeMax=1}}
 {id=Modulator_2, children [Mixer_1], properties {pitchMax=1, pitchMin=1, volumeMax=1, volumeMin=1}}
 {id=Modulator_3, children [Mixer_0], properties {pitchMax=1, pitchMin=1, volumeMax=0.25, volumeMin=0.25}}
 {id=Modulator_4, children [Mixer_0], properties {pitchMax=1, pitchMin=1, volumeMax=1, volumeMin=1}}
 {id=Modulator_5, children [Random_2], properties {pitchMax=1.2, pitchMin=0.9, volumeMax=1}}
 {id=Modulator_6, children [Random_5], properties {pitchMax=1.2, pitchMin=0.9, volumeMax=1}}
 {id=Modulator_8, children [Random_1], properties {pitchMax=1.2, pitchMin=0.9, volumeMax=1}}
 {id=Random_0, children [WavePlayer_0,WavePlayer_1,WavePlayer_2], properties {bShouldExcludeFromBranchCulling=true, weights [1,1,1]}}
 {
  id=Random_1
  children [WavePlayer_3,WavePlayer_4,WavePlayer_5,WavePlayer_6]
  properties {bShouldExcludeFromBranchCulling=true, weights [1,1,1,1]}
 }
 {
  id=Random_2
  children [WavePlayer_13, WavePlayer_14, WavePlayer_15, WavePlayer_16]
  properties {bShouldExcludeFromBranchCulling=true, weights [1,1,1,1]}
 }
 {
  id=Random_3
  children [WavePlayer_17, WavePlayer_18, WavePlayer_19, WavePlayer_20]
  properties {bShouldExcludeFromBranchCulling=true, weights [1,1,1,1]}
 }
 {id=Random_5, children [WavePlayer_11, WavePlayer_12, WavePlayer_21], properties {bShouldExcludeFromBranchCulling=true, weights [1,1,1]}}
 {id=WavePlayer_0, properties {soundWaveAssetPtr=&/IP_MotoWrestle_Attack_03_Hammer_Form_01}}
 {id=WavePlayer_1, properties {soundWaveAssetPtr=&/IP_MotoWrestle_Attack_03_Hammer_Form_02}}
 {id=WavePlayer_11, properties {soundWaveAssetPtr=&IP_MotoWrestle_Attack_03_Hammer_Form_01}}
 {id=WavePlayer_12, properties {soundWaveAssetPtr=&IP_MotoWrestle_Attack_03_Hammer_Form_02}}
 {id=WavePlayer_13, properties {soundWaveAssetPtr=&/IP_MotoWrestle_Attack_03_Swing_Up_01}}
 {id=WavePlayer_14, properties {soundWaveAssetPtr=&/IP_MotoWrestle_Attack_03_Swing_Up_02}}
 {id=WavePlayer_15, properties {soundWaveAssetPtr=&/IP_MotoWrestle_Attack_03_Swing_Up_03}}
 {id=WavePlayer_16, properties {soundWaveAssetPtr=&/IP_MotoWrestle_Attack_03_Swing_Up_04}}
 {id=WavePlayer_17, properties {soundWaveAssetPtr=&/IP_MotoWrestle_Attack_03_Hammer_Whsh_01}}
 {id=WavePlayer_18, properties {soundWaveAssetPtr=&/IP_MotoWrestle_Attack_03_Hammer_Whsh_02}}
 {id=WavePlayer_19, properties {soundWaveAssetPtr=&/IP_MotoWrestle_Attack_03_Hammer_Whsh_03}}
 {id=WavePlayer_2, properties {soundWaveAssetPtr=&/IP_MotoWrestle_Attack_03_Hammer_Form_03}}
 {id=WavePlayer_20, properties {soundWaveAssetPtr=&/IP_MotoWrestle_Attack_03_Hammer_Whsh_04}}
 {id=WavePlayer_21, properties {soundWaveAssetPtr=&IP_MotoWrestle_Attack_03_Hammer_Form_03}}
 {id=WavePlayer_3, properties {soundWaveAssetPtr=&IP_MotoWrestle_Attack_03_Swing_Up_01}}
 {id=WavePlayer_4, properties {soundWaveAssetPtr=&IP_MotoWrestle_Attack_03_Swing_Up_02}}
 {id=WavePlayer_5, properties {soundWaveAssetPtr=&IP_MotoWrestle_Attack_03_Swing_Up_03}}
 {id=WavePlayer_6, properties {soundWaveAssetPtr=&IP_MotoWrestle_Attack_03_Swing_Up_04}}
]
properties {
 bDebug=true
 bOverrideConcurrency=true
 maxDistance=2097152
 soundClassObject=&Fort_WeaponMelee
 volumeMultiplier=0.7
 attenuationOverrides {distanceAlgorithm=Logarithmic}
 concurrencyOverrides {maxCount=3}
}
