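# Sound cue for the MotoWrestle equip sound, classed as weapon foley.
# A Random node picks one of three equip wave variants with equal weight and is excluded from branch culling.
# It uses the shared weapon reload attenuation, slightly reduced volume, and overrides concurrency to cap simultaneous instances.
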SoundCue /MotoWrestleGameplay/Assets/Audio/IP_SC_MotoWrestle_Equip
root=Random_0
nodes [
 {id=Random_0, children [WavePlayer_1,WavePlayer_2,WavePlayer_3], properties {bShouldExcludeFromBranchCulling=true, weights [1,1,1]}}
 {id=WavePlayer_1, properties {soundWaveAssetPtr=&/IP_MotoWrestle_Equip_01}}
 {id=WavePlayer_2, properties {soundWaveAssetPtr=&/IP_MotoWrestle_Equip_02}}
 {id=WavePlayer_3, properties {soundWaveAssetPtr=&/IP_MotoWrestle_Equip_03}}
]
properties {
 attenuationSettings=&Fort_Weapon_Reload_Att
 bOverrideConcurrency=true
 maxDistance=2250
 soundClassObject=&Fort_WeaponFoley
 volumeMultiplier=0.85
 concurrencyOverrides {maxCount=2}
}
