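# Temporary MotoWrestle footstep SoundCue for freezing water, randomly playing one of ten combat-boot ice-run foley samples.
# An islocal branch routes local steps to first-person attenuation and others to an IsEnemy branch using third-person attenuation.
# Non-enemy remote steps play at half the modulator volume of enemy or unset ones, with all paths sharing one mixer and random pool.
# It uses the melee weapon sound class and local footstep concurrency.
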SoundCue /MotoWrestleGameplay/Assets/Audio/IP_SC_MotoWrestle_FS_FreezingWater_TEMP
root=Branch_0
nodes [
 {id=Attenuation_0, children [Modulator_2], properties {attenuationSettings=/Game/Sounds/Fort_Player/Footsteps/Settings/Fort_Player_FS_1P_Att}}
 {id=Attenuation_1, children [Modulator_3], properties {attenuationSettings=/Game/Sounds/Fort_Player/Footsteps/Settings/Fort_Player_FS_3P_Att}}
 {id=Attenuation_2, children [Modulator_4], properties {attenuationSettings=&Fort_Player_FS_3P_Att}}
 {id=Branch_0, children [Attenuation_0, Branch_3, Branch_3], properties {boolParameterName=islocal}}
 {id=Branch_3, children [Attenuation_2, Attenuation_1, Attenuation_2], properties {boolParameterName=IsEnemy}}
 {id=Mixer_0, children [Random_2,null], properties {inputVolume [1,0.5]}}
 {id=Modulator_2, children [Mixer_0], properties {pitchMax=1, pitchMin=1, volumeMax=1, volumeMin=1}}
 {id=Modulator_3, children [Mixer_0], properties {pitchMax=1, pitchMin=1, volumeMax=0.5, volumeMin=0.5}}
 {id=Modulator_4, children [Mixer_0], properties {pitchMax=1, pitchMin=1, volumeMax=1, volumeMin=1}}
 {
  id=Random_2
  children [
   WavePlayer_30
   WavePlayer_31
   WavePlayer_32
   WavePlayer_33
   WavePlayer_34
   WavePlayer_35
   WavePlayer_36
   WavePlayer_37
   WavePlayer_38
   WavePlayer_39
  ]
  properties {weights [1,1,1,1,1,1,1,1,1,1]}
 }
 {id=WavePlayer_30, properties {soundWaveAssetPtr=/MotoWrestleGameplay/Assets/Audio/TEMP/Foley_FS_1P_BootsCombat_Ice_Run_01}}
 {id=WavePlayer_31, properties {soundWaveAssetPtr=/MotoWrestleGameplay/Assets/Audio/TEMP/Foley_FS_1P_BootsCombat_Ice_Run_02}}
 {id=WavePlayer_32, properties {soundWaveAssetPtr=/MotoWrestleGameplay/Assets/Audio/TEMP/Foley_FS_1P_BootsCombat_Ice_Run_03}}
 {id=WavePlayer_33, properties {soundWaveAssetPtr=/MotoWrestleGameplay/Assets/Audio/TEMP/Foley_FS_1P_BootsCombat_Ice_Run_04}}
 {id=WavePlayer_34, properties {soundWaveAssetPtr=/MotoWrestleGameplay/Assets/Audio/TEMP/Foley_FS_1P_BootsCombat_Ice_Run_05}}
 {id=WavePlayer_35, properties {soundWaveAssetPtr=/MotoWrestleGameplay/Assets/Audio/TEMP/Foley_FS_1P_BootsCombat_Ice_Run_06}}
 {id=WavePlayer_36, properties {soundWaveAssetPtr=/MotoWrestleGameplay/Assets/Audio/TEMP/Foley_FS_1P_BootsCombat_Ice_Run_07}}
 {id=WavePlayer_37, properties {soundWaveAssetPtr=/MotoWrestleGameplay/Assets/Audio/TEMP/Foley_FS_1P_BootsCombat_Ice_Run_08}}
 {id=WavePlayer_38, properties {soundWaveAssetPtr=/MotoWrestleGameplay/Assets/Audio/TEMP/Foley_FS_1P_BootsCombat_Ice_Run_09}}
 {id=WavePlayer_39, properties {soundWaveAssetPtr=/MotoWrestleGameplay/Assets/Audio/TEMP/Foley_FS_1P_BootsCombat_Ice_Run_10}}
]
properties {
 attenuationSettings=/Game/Sounds/Attenuation/Fort_Player_Footsteps_1P_Walk_Athena_Att
 maxDistance=2097152
 soundClassObject=&Fort_WeaponMelee
 volumeMultiplier=0.5
 concurrencyOverrides {bLimitToOwner=true, maxCount=1, resolutionRule=StopOldest}
 concurrencySet [&Athena_Footstep_Local_Conc]
}
