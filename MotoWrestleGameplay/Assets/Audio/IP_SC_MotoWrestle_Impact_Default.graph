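# Sound cue for the default MotoWrestle impact sound, routed by whether the listener is local and whether the source is an enemy.
# An islocal branch picks a path without an attenuation asset, otherwise IsEnemy selects pickaxe player-impact attenuation, with non-enemy hits much quieter.
# All paths share a mixer layering random Default and pitched-up Player waves, three variants each.
# Global and local concurrency sets are assigned, and virtualization is disabled.
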SoundCue /MotoWrestleGameplay/Assets/Audio/IP_SC_MotoWrestle_Impact_Default
root=Branch_1
nodes [
 {id=Attenuation_0, children [Modulator_2]}
 {id=Attenuation_1, children [Modulator_3], properties {attenuationSettings=/Game/Sounds/Attenuation/Fort_Weapon_Pickaxe_Player_Impact_Att}}
 {id=Attenuation_2, children [Modulator_4], properties {attenuationSettings=&Fort_Weapon_Pickaxe_Player_Impact_Att}}
 {id=Branch_1, children [Attenuation_0, Branch_3, Branch_3], properties {boolParameterName=islocal}}
 {id=Branch_3, children [Attenuation_2, Attenuation_1, Attenuation_2], properties {boolParameterName=IsEnemy}}
 {id=Mixer_0, children [Modulator_5,Modulator_6], properties {inputVolume [1,1]}}
 {id=Modulator_2, children [Mixer_0], properties {pitchMax=1, pitchMin=1, volumeMax=1, volumeMin=1}}
 {id=Modulator_3, children [Mixer_0], properties {pitchMax=1, pitchMin=1, volumeMax=0.25, volumeMin=0.25}}
 {id=Modulator_4, children [Mixer_0], properties {pitchMax=1, pitchMin=1, volumeMax=1, volumeMin=1}}
 {id=Modulator_5, children [Random_2]}
 {id=Modulator_6, children [Random_3], properties {pitchMax=1.2, pitchMin=1.1, volumeMax=1, volumeMin=1}}
 {id=Random_2, children [WavePlayer_3,WavePlayer_4,WavePlayer_5], properties {bShouldExcludeFromBranchCulling=true, weights [1,1,1]}}
 {id=Random_3, children [WavePlayer_6,WavePlayer_7,WavePlayer_8], properties {bShouldExcludeFromBranchCulling=true, weights [1,1,1]}}
 {id=WavePlayer_3, properties {soundWaveAssetPtr=&/IP_MotoWrestle_Impact_Default_01}}
 {id=WavePlayer_4, properties {soundWaveAssetPtr=&/IP_MotoWrestle_Impact_Default_02}}
 {id=WavePlayer_5, properties {soundWaveAssetPtr=&/IP_MotoWrestle_Impact_Default_03}}
 {id=WavePlayer_6, properties {soundWaveAssetPtr=&/IP_MotoWrestle_Impact_Player_01}}
 {id=WavePlayer_7, properties {soundWaveAssetPtr=&/IP_MotoWrestle_Impact_Player_02}}
 {id=WavePlayer_8, properties {soundWaveAssetPtr=&/IP_MotoWrestle_Impact_Player_03}}
]
properties {
 bDebug=true
 maxDistance=2097152
 soundClassObject=/Game/Sounds/FortSoundClassesAndModes/Fort_Impact_Player_PSM
 virtualizationMode=Disabled
 volumeMultiplier=1.25
 attenuationOverrides {distanceAlgorithm=Logarithmic}
 concurrencyOverrides {maxCount=3}
 concurrencySet [&/IP_SC_MotoWrestle_Impacts_Global, &/IP_SC_MotoWrestle_Impacts_Local]
}
