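# Sound cue for the first-person (1P) hammer impact sound in the MotoWrestle game mode.
# A Random node picks one of three equally weighted hammer impact waves and feeds a Modulator for slight variation.
# It uses hit-marker receive attenuation and the player impact sound class, with virtualization disabled.
# It references global and local MotoWrestle impact concurrency sets, alongside override settings for a single stop-oldest voice.
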SoundCue /MotoWrestleGameplay/Assets/Audio/IP_SC_MotoWrestle_Impact_Hammer_1P
root=Modulator_1
nodes [
 {id=Modulator_1, children [Random_3]}
 {id=Random_3, children [WavePlayer_4, WavePlayer_11, WavePlayer_12], properties {bShouldExcludeFromBranchCulling=true, weights [1,1,1]}}
 {id=WavePlayer_11, properties {soundWaveAssetPtr=&/IP_MotoWrestle_Impact_Hammer_02}}
 {id=WavePlayer_12, properties {soundWaveAssetPtr=&/IP_MotoWrestle_Impact_Hammer_03}}
 {id=WavePlayer_4, properties {soundWaveAssetPtr=&/IP_MotoWrestle_Impact_Hammer_01}}
]
properties {
 attenuationSettings=/Game/Athena/Sounds/HitNotifications/Fort_HitMarker_Receive_Att
 maxDistance=2097152
 soundClassObject=/Game/Sounds/FortSoundClassesAndModes/Fort_Impact_Player_PSM
 virtualizationMode=Disabled
 volumeMultiplier=1.6
 concurrencyOverrides {maxCount=1, resolutionRule=StopOldest}
 concurrencySet [&/IP_SC_MotoWrestle_Impacts_Global, &/IP_SC_MotoWrestle_Impacts_Local]
}
