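# Third-person sound cue for MotoWrestle hammer impacts, randomly picking one of three equally weighted wave variations.
# An IsEnemy branch plays it at full volume when true or unset and at quarter volume when false.
# Both volume paths share the random source and use the pickaxe player-impact attenuation.
# Global and local impact concurrency sets plus a stop-oldest override limit stacking, and virtualization is disabled.
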SoundCue /MotoWrestleGameplay/Assets/Audio/IP_SC_MotoWrestle_Impact_Hammer_3P
root=Branch_2
nodes [
 {id=Attenuation_4, children [Modulator_0], properties {attenuationSettings=/Game/Sounds/Attenuation/Fort_Weapon_Pickaxe_Player_Impact_Att}}
 {id=Attenuation_5, children [Modulator_5], properties {attenuationSettings=&Fort_Weapon_Pickaxe_Player_Impact_Att}}
 {id=Branch_2, children [Attenuation_5, Attenuation_4, Attenuation_5], properties {boolParameterName=IsEnemy}}
 {id=Modulator_0, children [Modulator_2], properties {pitchMax=1, pitchMin=1, volumeMax=0.25, volumeMin=0.25}}
 {id=Modulator_2, children [Random_0], properties {volumeMax=1, volumeMin=1}}
 {id=Modulator_5, children [Modulator_2], properties {pitchMax=1, pitchMin=1, volumeMax=1, volumeMin=1}}
 {id=Random_0, children [WavePlayer_4, WavePlayer_11, WavePlayer_12], properties {bShouldExcludeFromBranchCulling=true, weights [1,1,1]}}
 {id=WavePlayer_11, properties {soundWaveAssetPtr=&/IP_MotoWrestle_Impact_Hammer_02}}
 {id=WavePlayer_12, properties {soundWaveAssetPtr=&/IP_MotoWrestle_Impact_Hammer_03}}
 {id=WavePlayer_4, properties {soundWaveAssetPtr=&/IP_MotoWrestle_Impact_Hammer_01}}
]
properties {
 maxDistance=3512
 soundClassObject=/Game/Sounds/FortSoundClassesAndModes/Fort_Impact_Player_PSM
 virtualizationMode=Disabled
 volumeMultiplier=1
 concurrencyOverrides {maxCount=3, resolutionRule=StopOldest}
 concurrencySet [&/IP_SC_MotoWrestle_Impacts_Global, &/IP_SC_MotoWrestle_Impacts_Local]
}
