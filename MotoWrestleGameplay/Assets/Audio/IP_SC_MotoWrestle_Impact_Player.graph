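# Sound Cue for player impacts in the MotoWrestle mode, with an empty graph: no root node or wave players, so the cue itself has no audio source.
# It carries playback settings: the Fort_Impact_Player_PSM sound class, boosted volume, disabled virtualization and a very large max distance.
# It references global and local MotoWrestle impact concurrency sets, alongside a small max-count value in its concurrency overrides.
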SoundCue /MotoWrestleGameplay/Assets/Audio/IP_SC_MotoWrestle_Impact_Player
root=null
nodes []
properties {
 maxDistance=2097152
 soundClassObject=/Game/Sounds/FortSoundClassesAndModes/Fort_Impact_Player_PSM
 virtualizationMode=Disabled
 volumeMultiplier=2
 concurrencyOverrides {maxCount=3}
 concurrencySet [&/IP_SC_MotoWrestle_Impacts_Global, &/IP_SC_MotoWrestle_Impacts_Local]
}
