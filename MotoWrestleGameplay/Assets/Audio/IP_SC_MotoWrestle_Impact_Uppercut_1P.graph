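# Sound cue for the 1P uppercut impact in MotoWrestle, played through the player impact sound class.
# It mixes a random crack with a quieter, briefly delayed random shards layer, each picked from three waves.
# A modulator on the mix adds slight upward pitch variation.
# Virtualization is disabled, and global and local MotoWrestle impact concurrency sets govern simultaneous plays.
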SoundCue /MotoWrestleGameplay/Assets/Audio/IP_SC_MotoWrestle_Impact_Uppercut_1P
root=Modulator_6
nodes [
 {id=Delay_0, children [Random_1], properties {delayMax=0.15, delayMin=0.15}}
 {id=Mixer_0, children [Random_0,Delay_0], properties {inputVolume [1,0.5]}}
 {id=Modulator_6, children [Mixer_0], properties {pitchMax=1.1, pitchMin=1, volumeMax=1, volumeMin=1}}
 {id=Random_0, children [WavePlayer_3,WavePlayer_4,WavePlayer_6], properties {bShouldExcludeFromBranchCulling=true, weights [1,1,1]}}
 {id=Random_1, children [WavePlayer_0,WavePlayer_1,WavePlayer_2], properties {bShouldExcludeFromBranchCulling=true, weights [1,1,1]}}
 {id=WavePlayer_0, properties {soundWaveAssetPtr=&/IP_MotoWrestle_Impact_Uppercut_Shards_01}}
 {id=WavePlayer_1, properties {soundWaveAssetPtr=&/IP_MotoWrestle_Impact_Uppercut_Shards_02}}
 {id=WavePlayer_2, properties {soundWaveAssetPtr=&/IP_MotoWrestle_Impact_Uppercut_Shards_03}}
 {id=WavePlayer_3, properties {soundWaveAssetPtr=&/IP_MotoWrestle_Uppercut_Impact_Crack_01}}
 {id=WavePlayer_4, properties {soundWaveAssetPtr=&/IP_MotoWrestle_Uppercut_Impact_Crack_02}}
 {id=WavePlayer_6, properties {soundWaveAssetPtr=&/IP_MotoWrestle_Uppercut_Impact_Crack_03}}
]
properties {
 maxDistance=2097152
 soundClassObject=/Game/Sounds/FortSoundClassesAndModes/Fort_Impact_Player_PSM
 virtualizationMode=Disabled
 volumeMultiplier=1.85
 concurrencyOverrides {maxCount=3}
 concurrencySet [&/IP_SC_MotoWrestle_Impacts_Global, &/IP_SC_MotoWrestle_Impacts_Local]
}
