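# Third-person sound cue for the MotoWrestle uppercut impact, playing one of three equally weighted wave variants at random.
# An IsEnemy branch plays it at full modulator volume when true or unset, and at quarter volume when false.
# Both paths share the random selection and use the pickaxe player-impact attenuation.
# Global and local MotoWrestle impact concurrency sets limit overlapping hits, and virtualization is disabled.
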SoundCue /MotoWrestleGameplay/Assets/Audio/IP_SC_MotoWrestle_Impact_Uppercut_3P
root=Branch_2
nodes [
 {id=Attenuation_4, children [Modulator_1], properties {attenuationSettings=/Game/Sounds/Attenuation/Fort_Weapon_Pickaxe_Player_Impact_Att}}
 {id=Attenuation_5, children [Modulator_5], properties {attenuationSettings=&Fort_Weapon_Pickaxe_Player_Impact_Att}}
 {id=Branch_2, children [Attenuation_5, Attenuation_4, Attenuation_5], properties {boolParameterName=IsEnemy}}
 {id=Modulator_1, children [Modulator_2], properties {pitchMax=1, pitchMin=1, volumeMax=0.25, volumeMin=0.25}}
 {id=Modulator_2, children [Random_0], properties {volumeMax=1, volumeMin=1}}
 {id=Modulator_5, children [Modulator_2], properties {pitchMax=1, pitchMin=1, volumeMax=1, volumeMin=1}}
 {id=Random_0, children [WavePlayer_0,WavePlayer_1,WavePlayer_2], properties {bShouldExcludeFromBranchCulling=true, weights [1,1,1]}}
 {id=WavePlayer_0, properties {soundWaveAssetPtr=&/IP_MotoWrestle_Impact_Uppercut_01}}
 {id=WavePlayer_1, properties {soundWaveAssetPtr=&/IP_MotoWrestle_Impact_Uppercut_02}}
 {id=WavePlayer_2, properties {soundWaveAssetPtr=&/IP_MotoWrestle_Impact_Uppercut_03}}
]
properties {
 maxDistance=3512
 soundClassObject=/Game/Sounds/FortSoundClassesAndModes/Fort_Impact_Player_PSM
 virtualizationMode=Disabled
 volumeMultiplier=1.25
 concurrencyOverrides {maxCount=3}
 concurrencySet [&/IP_SC_MotoWrestle_Impacts_Global, &/IP_SC_MotoWrestle_Impacts_Local]
}
