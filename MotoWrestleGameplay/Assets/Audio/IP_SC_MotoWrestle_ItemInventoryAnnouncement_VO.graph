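# MotoWrestle SoundCue that plays an announcer voice line, named for item inventory announcements.
# A single wave player feeds the root with a PoppySoap announcer MotoWrestle recording at slightly reduced volume.
# It uses the non-critical in-game UI sound class with a very large max distance.
# An announcer equips concurrency set is assigned, alongside override settings specifying a single instance that prevents new plays.
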SoundCue /MotoWrestleGameplay/Assets/Audio/IP_SC_MotoWrestle_ItemInventoryAnnouncement_VO
root=WavePlayer_0
nodes [{id=WavePlayer_0, properties {soundWaveAssetPtr=/PoppySoap/Assets/Sound/IP_PoppySoap_VO_Announcer_MotoWrestle_01}}]
properties {
 maxDistance=2097152
 soundClassObject=&Fort_UI_InGame_NonCritical
 volumeMultiplier=0.85
 concurrencyOverrides {maxCount=1, resolutionRule=PreventNew}
 concurrencySet [/PoppySoap/Assets/Sound/CONC_IP_PoppySoap_VO_Announcer_Equips]
}
