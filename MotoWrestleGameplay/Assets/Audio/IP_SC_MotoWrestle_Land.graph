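# Sound cue for the MotoWrestle landing drop foley.
# A single wave player feeds one drop-land wave straight to the output, with no randomization or modulation.
# It uses Fortnite weapon-drop attenuation and the weapon foley sound class, with a concurrency override capping simultaneous instances to a small number.
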SoundCue /MotoWrestleGameplay/Assets/Audio/IP_SC_MotoWrestle_Land
root=WavePlayer_0
nodes [{id=WavePlayer_0, properties {soundWaveAssetPtr=&/IP_MotoWrestle_Drop_Land_01}}]
properties {
 attenuationSettings=&Fort_Weapon_Drop_Att
 bOverrideConcurrency=true
 maxDistance=3000
 soundClassObject=&Fort_WeaponFoley
 concurrencyOverrides {maxCount=2}
}
