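# Temporary MotoWrestle sound cue for landing in freezing water, routed to the Fort_Weapons sound class.
# A shared mixer layers a random ice boot-landing foley at full volume with a random water landing at half volume.
# An islocal branch selects first-person attenuation, otherwise an IsEnemy branch applies third-person attenuation.
# Non-enemy remote landings play at half volume, while local, enemy and unset IsEnemy cases stay at full volume.
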SoundCue /MotoWrestleGameplay/Assets/Audio/IP_SC_MotoWrestle_Land_FreezingWater_TEMP
root=Branch_0
nodes [
 {id=Attenuation_0, children [Modulator_2], properties {attenuationSettings=/Game/Sounds/Fort_Player/Footsteps/Settings/Fort_Player_FS_1P_Att}}
 {id=Attenuation_1, children [Modulator_3], properties {attenuationSettings=/Game/Sounds/Fort_Player/Footsteps/Settings/Fort_Player_FS_3P_Att}}
 {id=Attenuation_2, children [Modulator_4], properties {attenuationSettings=&Fort_Player_FS_3P_Att}}
 {id=Branch_0, children [Attenuation_0, Branch_3, Branch_3], properties {boolParameterName=islocal}}
 {id=Branch_3, children [Attenuation_2, Attenuation_1, Attenuation_2], properties {boolParameterName=IsEnemy}}
 {id=Mixer_0, children [Random_0,Random_1], properties {inputVolume [1,0.5]}}
 {id=Modulator_2, children [Mixer_0], properties {pitchMax=1, pitchMin=1, volumeMax=1, volumeMin=1}}
 {id=Modulator_3, children [Mixer_0], properties {pitchMax=1, pitchMin=1, volumeMax=0.5, volumeMin=0.5}}
 {id=Modulator_4, children [Mixer_0], properties {pitchMax=1, pitchMin=1, volumeMax=1, volumeMin=1}}
 {
  id=Random_0
  children [WavePlayer_0,WavePlayer_1,WavePlayer_2,WavePlayer_3,WavePlayer_4,WavePlayer_5,WavePlayer_6,WavePlayer_7]
  properties {weights [1,1,1,1,1,1,1,1]}
 }
 {
  id=Random_1
  children [WavePlayer_8, WavePlayer_9, WavePlayer_10, WavePlayer_11, WavePlayer_12, WavePlayer_13]
  properties {bShouldExcludeFromBranchCulling=true, weights [1,1,1,1,1,1]}
 }
 {id=WavePlayer_0, properties {soundWaveAssetPtr=/MotoWrestleGameplay/Assets/Audio/TEMP/Foley_FS_1P_BootsCombat_Ice_Land_08}}
 {id=WavePlayer_1, properties {soundWaveAssetPtr=/MotoWrestleGameplay/Assets/Audio/TEMP/Foley_FS_1P_BootsCombat_Ice_Land_07}}
 {id=WavePlayer_10, properties {soundWaveAssetPtr=&/IP_MotoWrestle_Water_Navigation_Land_03}}
 {id=WavePlayer_11, properties {soundWaveAssetPtr=&/IP_MotoWrestle_Water_Navigation_Land_04}}
 {id=WavePlayer_12, properties {soundWaveAssetPtr=&/IP_MotoWrestle_Water_Navigation_Land_05}}
 {id=WavePlayer_13, properties {soundWaveAssetPtr=&/IP_MotoWrestle_Water_Navigation_Land_06}}
 {id=WavePlayer_2, properties {soundWaveAssetPtr=/MotoWrestleGameplay/Assets/Audio/TEMP/Foley_FS_1P_BootsCombat_Ice_Land_06}}
 {id=WavePlayer_3, properties {soundWaveAssetPtr=/MotoWrestleGameplay/Assets/Audio/TEMP/Foley_FS_1P_BootsCombat_Ice_Land_05}}
 {id=WavePlayer_4, properties {soundWaveAssetPtr=/MotoWrestleGameplay/Assets/Audio/TEMP/Foley_FS_1P_BootsCombat_Ice_Land_04}}
 {id=WavePlayer_5, properties {soundWaveAssetPtr=/MotoWrestleGameplay/Assets/Audio/TEMP/Foley_FS_1P_BootsCombat_Ice_Land_03}}
 {id=WavePlayer_6, properties {soundWaveAssetPtr=/MotoWrestleGameplay/Assets/Audio/TEMP/Foley_FS_1P_BootsCombat_Ice_Land_02}}
 {id=WavePlayer_7, properties {soundWaveAssetPtr=/MotoWrestleGameplay/Assets/Audio/TEMP/Foley_FS_1P_BootsCombat_Ice_Land_01}}
 {id=WavePlayer_8, properties {soundWaveAssetPtr=&/IP_MotoWrestle_Water_Navigation_Land_01}}
 {id=WavePlayer_9, properties {soundWaveAssetPtr=&/IP_MotoWrestle_Water_Navigation_Land_02}}
]
properties {maxDistance=3584, soundClassObject=/Game/Sounds/FortSoundClassesAndModes/Fort_Weapons}
