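# Sound cue for the MotoWrestle slide dash air exit, layering a slide-end whoosh with a random pick of three metal slide sounds.
# Branches on islocal, then IsEnemy, to give local, enemy and non-enemy listeners separate volume and attenuation paths.
# Local and non-enemy paths share a mix with a quieter whoosh, non-enemy much lower overall; enemy or unset uses a full-level mix.
# Concurrency override caps simultaneous instances.
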SoundCue /MotoWrestleGameplay/Assets/Audio/IP_SC_MotoWrestle_SlideDash_AirExit
root=Branch_0
nodes [
 {id=Attenuation_0, children [Modulator_0]}
 {
  id=Attenuation_1
  children [Modulator_3]
  properties {attenuationSettings=/Game/Sounds/Attenuation/Fort_Weapon_Melee_Attack_Close_Range_Stereo_Att}
 }
 {
  id=Attenuation_2
  children [Modulator_4]
  properties {attenuationSettings=&Fort_Weapon_Melee_Attack_Close_Range_Stereo_Att}
 }
 {id=Branch_0, children [Attenuation_0, Branch_3, Branch_3], properties {boolParameterName=islocal}}
 {id=Branch_3, children [Attenuation_2, Attenuation_1, Attenuation_2], properties {boolParameterName=IsEnemy}}
 {id=Mixer_0, children [Modulator_1,Modulator_2], properties {inputVolume [0.35,1]}}
 {id=Mixer_1, children [Modulator_5,Modulator_6], properties {inputVolume [1,1]}}
 {id=Modulator_0, children [Mixer_0], properties {pitchMax=1, pitchMin=1, volumeMax=0.9, volumeMin=0.9}}
 {id=Modulator_1, children [WavePlayer_0], properties {pitchMax=1.2}}
 {id=Modulator_2, children [Random_0]}
 {id=Modulator_3, children [Mixer_0], properties {pitchMax=1, pitchMin=1, volumeMax=0.25, volumeMin=0.25}}
 {id=Modulator_4, children [Mixer_1], properties {pitchMax=1, pitchMin=1, volumeMax=1, volumeMin=1}}
 {id=Modulator_5, children [WavePlayer_1], properties {pitchMax=1.2}}
 {id=Modulator_6, children [Random_1]}
 {id=Random_0, children [WavePlayer_3,WavePlayer_4,WavePlayer_5], properties {bShouldExcludeFromBranchCulling=true, weights [1,1,1]}}
 {id=Random_1, children [WavePlayer_2,WavePlayer_6,WavePlayer_7], properties {bShouldExcludeFromBranchCulling=true, weights [1,1,1]}}
 {id=WavePlayer_0, properties {soundWaveAssetPtr=&/IP_MotoWrestle_SlideDash_Slide_End_Whsh_01}}
 {id=WavePlayer_1, properties {soundWaveAssetPtr=&IP_MotoWrestle_SlideDash_Slide_End_Whsh_01}}
 {id=WavePlayer_2, properties {soundWaveAssetPtr=&/IP_MotoWrestle_SlideDash_Slide_Metal_01}}
 {id=WavePlayer_3, properties {soundWaveAssetPtr=&IP_MotoWrestle_SlideDash_Slide_Metal_01}}
 {id=WavePlayer_4, properties {soundWaveAssetPtr=&/IP_MotoWrestle_SlideDash_Slide_Metal_02}}
 {id=WavePlayer_5, properties {soundWaveAssetPtr=&/IP_MotoWrestle_SlideDash_Slide_Metal_03}}
 {id=WavePlayer_6, properties {soundWaveAssetPtr=&IP_MotoWrestle_SlideDash_Slide_Metal_02}}
 {id=WavePlayer_7, properties {soundWaveAssetPtr=&IP_MotoWrestle_SlideDash_Slide_Metal_03}}
]
properties {
 attenuationSettings=&Fort_Weapon_Melee_Attack_Close_Range_Stereo_Att
 bOverrideConcurrency=true
 maxDistance=2097152
 soundClassObject=&Fort_WeaponMelee
 volumeMultiplier=1
 concurrencyOverrides {maxCount=3}
}
