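# Sound cue for the MotoWrestle slide dash ground exit, currently an empty shell with no root or sound nodes, so it produces no audio itself.
# It uses the Fortnite melee weapon sound class and a close-range stereo melee attack attenuation asset.
# A concurrency override caps simultaneous instances at a small count, limited per owner.
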SoundCue /MotoWrestleGameplay/Assets/Audio/IP_SC_MotoWrestle_SlideDash_GroundExit
root=null
nodes []
properties {
 attenuationSettings=/Game/Sounds/Attenuation/Fort_Weapon_Melee_Attack_Close_Range_Stereo_Att
 bOverrideConcurrency=true
 maxDistance=2097152
 soundClassObject=&Fort_WeaponMelee
 volumeMultiplier=1
 concurrencyOverrides {bLimitToOwner=true, maxCount=3}
}
