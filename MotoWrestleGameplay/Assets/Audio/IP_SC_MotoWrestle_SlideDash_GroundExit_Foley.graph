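# Sound cue named for the foley of a MotoWrestle slide-dash ground exit, currently with no nodes or root, so it plays no audio.
# It is configured with the Fort weapon foley sound class and borrows the weapon pickup attenuation asset.
# A concurrency override caps it at two simultaneous instances.
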SoundCue /MotoWrestleGameplay/Assets/Audio/IP_SC_MotoWrestle_SlideDash_GroundExit_Foley
root=null
nodes []
properties {
 attenuationSettings=&Fort_Weapon_Pickup_Att
 bOverrideConcurrency=true
 maxDistance=3500
 soundClassObject=&Fort_WeaponFoley
 concurrencyOverrides {maxCount=2}
}
