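# Sound cue for the start of a slide dash in MotoWrestle, randomly picking one of three waves.
# Nested branches on islocal and IsEnemy set volume per listener: medium for local, full for enemy or unset enemy state, quiet for non-enemy others.
# The non-local paths share the melee special dash attenuation asset.
# Concurrency uses global and local dash sets, plus a one-per-owner stop-oldest override block.
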SoundCue /MotoWrestleGameplay/Assets/Audio/IP_SC_MotoWrestle_SlideDash_Start
root=Branch_0
nodes [
 {id=Attenuation_0, children [Modulator_2]}
 {
  id=Attenuation_1
  children [Modulator_3]
  properties {attenuationSettings=/Game/Sounds/Attenuation/Fort_Weapon_Melee_Attack_Special_Dash_Att}
 }
 {
  id=Attenuation_2
  children [Modulator_4]
  properties {attenuationSettings=&Fort_Weapon_Melee_Attack_Special_Dash_Att}
 }
 {id=Branch_0, children [Attenuation_0, Branch_3, Branch_3], properties {boolParameterName=islocal}}
 {id=Branch_3, children [Attenuation_2, Attenuation_1, Attenuation_2], properties {boolParameterName=IsEnemy}}
 {id=Modulator_2, children [Random_0], properties {pitchMax=1, pitchMin=1, volumeMax=0.65, volumeMin=0.65}}
 {id=Modulator_3, children [Random_0], properties {pitchMax=1, pitchMin=1, volumeMax=0.25, volumeMin=0.25}}
 {id=Modulator_4, children [Random_0], properties {pitchMax=1, pitchMin=1, volumeMax=1, volumeMin=1}}
 {id=Random_0, children [WavePlayer_0,WavePlayer_2,WavePlayer_3], properties {bShouldExcludeFromBranchCulling=true, weights [1,1,1]}}
 {id=WavePlayer_0, properties {soundWaveAssetPtr=&/IP_MotoWrestle_SlideDash_Start_01}}
 {id=WavePlayer_2, properties {soundWaveAssetPtr=&/IP_MotoWrestle_SlideDash_Start_02}}
 {id=WavePlayer_3, properties {soundWaveAssetPtr=&/IP_MotoWrestle_SlideDash_Start_03}}
]
properties {
 maxDistance=2097152
 soundClassObject=&Fort_WeaponMelee
 volumeMultiplier=1
 concurrencyOverrides {bLimitToOwner=true, maxCount=1, resolutionRule=StopOldest}
 concurrencySet [&/IP_MotoWrestle_Dash_Global, &/IP_MotoWrestle_Dash_Local]
}
