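; MotoWrestle cosmetic part modifier that places ambient and idle VFX on the player's character.
; On EventOnPostResIn it runs the parent logic, then waits two ticks before adjusting effects.
; It spawns the MotoWrestle ambient Niagara system on the left lower arm bone with an offset, setting its rotation on success.
; It also repositions and rotates the existing idle VFX component, fetched through GetIdleVFXComponent, with a mirrored offset.
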(event EventOnPostResIn
  (bind _returnvalue (CallFunction|GetIdleVFXComponent))
  (|CallParentFunction)
  (Utilities|FlowControl|DelayUntilNextTick)
  (Utilities|FlowControl|DelayUntilNextTick)
  (Utilities|FlowControl|Sequence
    (:then_0
      (bind _returnvalue_2 (PartModifier|SpawnSystem|SpawnSystemAttachedPartModifier "/MotoWrestleGameplay/Assets/VFX/NS_MotoWrestle_AmbientFX.NS_MotoWrestle_AmbientFX" (CustomCharacter|GetSkeletalMeshForPartOnPlayerPawn) "lowerarm_l" "20.000000,0.000000,0.000000" "90.000000,0.000000,0.000000" "SnapToTarget")
        (:Success
          (Transformation|SetRelativeRotation _returnvalue_2 "90.000000,0.000000,0.000000" false false))))
    (:then_1
      (Transformation|SetRelativeLocation _returnvalue "-20.000000,0.000000,0.000000" false false)
      (Transformation|SetRelativeRotation _returnvalue "-90.000000,0.000000,0.000000" false false))))

(fn ConstructionScript
  (|CallParentFunction))
