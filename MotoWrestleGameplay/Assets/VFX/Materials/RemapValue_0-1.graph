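# Material function in the Math library category that remaps a scalar from an Input Low to Input High range into 0-1, stored with the VFX materials.
# It subtracts the low bound from the input, divides by the difference between the high and low bounds, and saturates the result.
# There is no guard against equal bounds, which would make the divisor zero.
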MaterialFunction /MotoWrestleGameplay/Assets/VFX/Materials/RemapValue_0-1
graph {
 properties {description=Remaps two values to 0-1 clamped, libraryCategoriesText [Math]}
 nodes [
  n1 = FunctionInput() {inputName=Input Low, inputType=FunctionInput_Scalar, sortPriority=1}
  n2 = Subtract(A=FunctionInput() {inputName=Input, inputType=FunctionInput_Scalar}, B=@n1)
  n3 = Subtract(A=FunctionInput() {inputName=Input High, inputType=FunctionInput_Scalar, sortPriority=2, previewValue=<1 0 0 1>}, B=@n1)
  FunctionOutput(None=Saturate(None=Divide(A=@n2, B=@n3)))
 ]
}
