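# Translucent unlit material for Niagara mesh particles that fakes an ice decal, by its name for a hand effect.
# World-space-aligned screen UVs around a camera-offset center sample ice textures for hue-shifted, tinted emissive and remapped opacity.
# A sphere mask on the position behind translucency, depth fade and particle alpha shape opacity, with dynamic parameters driving offset, hardness and scale.
# Custom depth limits it to marked objects, while PC-Perf, Android and Switch get zero opacity.
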Material /MotoWrestleGameplay/Assets/VFX/Materials/M_HandIceDecal
graph {
 properties {
  bAutomaticallySetUsageInEditor=false
  bUsedWithNiagaraMeshParticles=true
  blendMode=BLEND_Translucent
  shadingModel=MSM_Unlit
  twoSided=true
  usedShadingModels=Unlit
 }
 nodes [
  n1 = TextureObjectParameter() {group=IceTexture, parameterName=IceTexture_Texture, sortPriority=0, texture=&/Ice_D}
  n2 = VectorParameter() {group=IceTexture, parameterName=IceTexture_UVControl, channelNames {a=VPan, b=UPan, g=VScale, r=UScale}, defaultValue=<rgba 70 70 0 0>}
  n3 = DynamicParameter() {paramNames [CameraOffset,Hardness,UVScale,Param4]}
  n4 = MaterialFunctionCall() {materialFunction=&CameraDirectionVector}
  n5 = Add(
   A=Multiply(A=@n3.CameraOffset, B=Transform(None=@n4.Result) {transformSourceType=TRANSFORMSOURCE_World, transformType=TRANSFORM_Instance})
   B=Constant3Vector()
  )
  n6 = Multiply(
   A=MaterialFunctionCall!(Center Point World Position=TransformPosition(None=@n5) {transformSourceType=TRANSFORMPOSSOURCE_Instance, transformType=TRANSFORMPOSSOURCE_World}) {materialFunction=/Engine/Functions/Engine_MaterialFunctions02/UVs/WorldSpaceAlignedScreenCoordinates}.Offset and Scaled UVS
   B=@n3.UVScale
  )
  n7 = MaterialFunctionCall(
   Hue Shift Percentage=ScalarParameter() {defaultValue=0.0704, parameterName=ssT}
   Texture=MaterialFunctionCall(Texture=@n1, UVControl_RGB=@n2, UVControl_A=@n2, "[Override]UVInput"=@n6) {materialFunction=&FXCommon_TextureSample}.RGB
  ) {materialFunction=&HueShift}
  n8 = ScalarParameter() {parameterName=DS}
  MaterialFunctionCall(Emissive=Multiply(A=Desaturation(None=@n7.Result, Fraction=@n8), B=VectorParameter() {parameterName=Color, defaultValue=<rgba 4 4 4 1>}.RGB)) {materialFunction=&MF_FX_EmissiveAdjustments}
  n9 = TextureObject() {texture=&/IceD}
  n10 = MaterialFunctionCall!(
   Input=Power(
    Base=MaterialFunctionCall(Texture=@n9, UVControl_RGB=@n2, UVControl_A=@n2, "[Override]UVInput"=@n6) {materialFunction=&FXCommon_TextureSample}.R
    Exp=ScalarParameter() {defaultValue=0.1486, parameterName=Te}
   )
   Input Low=ScalarParameter() {defaultValue=0.1672, parameterName=IL}
   Input High=ScalarParameter() {defaultValue=0.7056, parameterName=IG}
  ) {materialFunction=&/RemapValue_0-1}
  n11 = MaterialFunctionCall() {materialFunction=&WorldPositionBehindTranslucency}
  n12 = ScalarParameter() {defaultValue=2, parameterName=DF}
  n13 = Multiply(
   A=SphereMask(
    A=TransformPosition(None=@n11.Result) {transformSourceType=TRANSFORMPOSSOURCE_World, transformType=TRANSFORMPOSSOURCE_Instance}
    B=@n5
    Hardness=@n3.Hardness
   ) {attenuationRadius=1, hardnessPercent=10}
   B=Power(Base=DepthFade(FadeDistance=@n12))
  )
  n14 = Divide(A=SceneTexture() {sceneTextureId=PPI_CustomDepth}.Color) {constB=40960}
  n15 = OneMinus(None=Saturate(None=Floor(None=@n14)))
  n16 = Constant()
  Multiply(
   A=Multiply(A=Multiply(A=@n10.Result, B=@n13), B=ParticleColor().A)
   B=MaterialFunctionCall("Xb1,PS4"=@n15, "PC-High&&Epic,XSX,PS5"=@n15, "PC-Perf,Android"=@n16, Switch=@n16, "PC-Low(NotPerf)&&Medium"=@n15) {materialFunction=&PlatformSwitch_1D}.Result
  )
  VectorParameter() {parameterName=Color, defaultValue=<rgba 4 4 4 1>}
 ]
}
