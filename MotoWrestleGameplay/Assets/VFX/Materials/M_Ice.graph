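# Translucent unlit ice material for Niagara mesh particles, with pixel normal offset refraction mode.
# A flattened normal drives view-aligned fake reflection UVs, optionally tinted by particle color, added to a desaturated, hue-shifted ice texture tinted by time of day.
# That sum is scaled into an emissive adjustment function.
# An alpha mask shares the ice UVs, offset by a dynamic parameter, and combines depth fade with particle alpha.
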Material /MotoWrestleGameplay/Assets/VFX/Materials/M_Ice
graph {
 properties {
  bAutomaticallySetUsageInEditor=false
  bUsedWithNiagaraMeshParticles=true
  blendMode=BLEND_Translucent
  refractionMethod=RM_PixelNormalOffset
  shadingModel=MSM_Unlit
  usedShadingModels=Unlit
 }
 nodes [
  n1 = TextureObjectParameter() {group=BumpIce, parameterName=BumpIce_Texture, sortPriority=0, texture=&/Ice_D}
  n2 = VectorParameter() {group=Ref, parameterName=Ref_UVControl, channelNames {a=VPan, b=UPan, g=VScale, r=UScale}, defaultValue=<rgba 0.7 0.7 0 0>}
  n3 = TextureObjectParameter() {
   group=MainNormal
   parameterName=MainNormal_Texture
   samplerType=SAMPLERTYPE_Normal
   sortPriority=0
   texture=/Game/Environments/Props/Rocks/Textures/T_Corrupt_Rock_AB_N
  }
  n4 = VectorParameter() {group=MainNormal, parameterName=MainNormal_UVControl, channelNames {a=VPan, b=UPan, g=VScale, r=UScale}, defaultValue=<rgba 1 1 0 0>}
  n5 = MaterialFunctionCall(
   Normal=MaterialFunctionCall(Texture=@n3, UVControl_RGB=@n4, UVControl_A=@n4) {materialFunction=&FXCommon_TextureSample}.RGB
   Flatness=ScalarParameter() {parameterName=f2}
  ) {materialFunction=&FlattenNormal}
  n6 = ReflectionVectorWS(CustomWorldNormal=Transform(None=@n5.Result))
  n7 = Multiply(
   A=MaterialFunctionCall(
    Texture=@n1
    UVControl_RGB=@n2
    UVControl_A=@n2
    "[Override]UVInput"=MaterialFunctionCall!(ReflectionVector=@n6, ReflectionTexture=@n1) {materialFunction=/Engine/Functions/Engine_MaterialFunctions01/Reflections/ViewAlignedReflection}.UVs
   ) {materialFunction=&FXCommon_TextureSample}.RGB
   B=ScalarParameter() {defaultValue=2, parameterName=a2t}
  )
  n8 = ParticleColor()
  n9 = LinearInterpolate(A=@n7, B=Multiply(A=@n8.RGB, B=@n7), Alpha=ScalarParameter() {parameterName=EnablePC})
  n10 = VectorParameter() {parameterName=Color_Morning, defaultValue=<rgba 1 0.8676 0.8 1>}
  n11 = ScalarParameter() {defaultValue=0.08, parameterName=seh}
  n12 = VectorParameter() {group=Ice2, parameterName=Ice2_UVControl, channelNames {a=VPan, b=UPan, g=VScale, r=UScale}, defaultValue=<rgba 1 1 0 0>}
  n13 = MaterialFunctionCall(
   Texture=TextureObjectParameter() {group=Ice2, parameterName=Ice2_Texture, sortPriority=0, texture=&Ice_D}
   AdditionalUVOffset=DynamicParameter() {paramNames [R,Param2,Param3,Param4]}.R
   UVControl_RGB=@n12
   UVControl_A=@n12
  ) {materialFunction=&FXCommon_TextureSample}
  n14 = Desaturation(None=@n13.RGB, Fraction=ScalarParameter() {defaultValue=0.8, parameterName=Desat})
  n15 = Multiply(
   A=MaterialFunctionCall(
    Color_Morning=@n10.RGB
    Color_Day=VectorParameter() {parameterName=Color_Day, defaultValue=<rgba 1 1 1 1>}.RGB
    Color_Evening=@n10.RGB
    Color_Night=VectorParameter() {parameterName=Color_Night, defaultValue=<rgba 0.65 0.7498 1 1>}.RGB
   ) {materialFunction=&Color_Time_of_Day}.Result
   B=Multiply(
    A=MaterialFunctionCall(Hue Shift Percentage=@n11, Texture=@n14) {materialFunction=&HueShift}.Result
    B=ScalarParameter() {defaultValue=2.5, parameterName=aef}
   )
  )
  MaterialFunctionCall(Emissive=Divide(A=Add(A=@n15, B=@n9)) {constB=1.97}) {materialFunction=&MF_FX_EmissiveAdjustments}
  n16 = TextureObjectParameter() {group=MainT, parameterName=MainT_Texture, sortPriority=0, texture=/Game/Environments/Props/Rocks/Textures/T_Corrupt_Rock_AB_D}
  n17 = VectorParameter() {group=MainT, parameterName=MainT_UVControl, channelNames {a=VPan, b=UPan, g=VScale, r=UScale}, defaultValue=<rgba 1 1 0 0>}
  Multiply(
   A=MaterialFunctionCall(Texture=@n16, UVControl_RGB=@n17, UVControl_A=@n17) {materialFunction=&FXCommon_TextureSample}.R
   B=VectorParameter() {parameterName=C, defaultValue=<rgba 10 10 10 1>}.RGB
  )
  n18 = Power(
   Base=TextureSample(UVs=@n13.UV) {texture=&/IceD}.R
   Exp=ScalarParameter() {defaultValue=0.2, parameterName=ssS}
  )
  n19 = ScalarParameter() {defaultValue=1.6658}
  Multiply(A=Multiply(A=Multiply(A=@n18, B=@n19), B=DepthFade() {fadeDistanceDefault=20}), B=@n8.A)
  Reroute(None=@n5.Result)
  ScalarParameter() {defaultValue=1.2678, parameterName=R}
  TextureSample() {texture=&Ice_D}
  TextureSample() {samplerType=SAMPLERTYPE_Normal, texture=&/IceN}
 ]
}
