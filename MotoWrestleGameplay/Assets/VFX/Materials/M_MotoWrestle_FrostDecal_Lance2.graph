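# Translucent unlit Niagara mesh-particle material that fakes a lance-shaped frost decal on scene geometry for MotoWrestle.
# Reconstructs world position from scene depth, building a box gradient from particle position and dynamic axis-angle and size parameters.
# World-XY ice texture, tinted by time of day, drives emissive.
# Opacity combines dissolving noise, lance mask, particle alpha, depth fade, Z-edge and derivative-based slope masks, plus CustomDepth masking on some platforms.
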Material /MotoWrestleGameplay/Assets/VFX/Materials/M_MotoWrestle_FrostDecal_Lance2
graph {
 properties {
  bAutomaticallySetUsageInEditor=false
  bUseTranslucencyVertexFog=false
  bUsedWithNiagaraMeshParticles=true
  blendMode=BLEND_Translucent
  shadingModel=MSM_Unlit
  twoSided=true
  usedShadingModels=Unlit
 }
 nodes [
  n1 = VectorParameter() {parameterName=Color_Morning, defaultValue=<rgba 1 0.8676 0.8 1>}
  n2 = VectorParameter() {parameterName=Color_Day, defaultValue=<rgba 1 1 1 1>}
  n3 = VectorParameter() {parameterName=Color_Night, defaultValue=<rgba 0.65 0.7498 1 1>}
  n4 = VectorParameter() {group=IceT, parameterName=IceT_UVControl, channelNames {a=VPan, b=UPan, g=VScale, r=UScale}, defaultValue=<rgba 2 2 0 0>}
  n5 = MaterialFunctionCall(SceneDepth=SceneDepth().R) {materialFunction=&WorldPositionBehindTranslucency}
  n6 = ComponentMask(None=@n5.Result) {g=true, r=true}
  n7 = MaterialFunctionCall(
   Texture=TextureObject() {texture=/Game/Packages/Fortnite_UniqueMaterials_1/Textures/T_Ice_Lake_Snow_D}
   UVControl_RGB=@n4
   UVControl_A=@n4
   "[Override]UVInput"=Rotator(Coordinate=Multiply(A=@n6) {constB=0.003}, Time=ScalarParameter() {defaultValue=0.25, parameterName=R})
  ) {materialFunction=&FXCommon_TextureSample}
  n8 = VectorParameter() {parameterName=WF, defaultValue=<rgba 2.55 2.7381 3 1>}
  MaterialFunctionCall(Emissive=Multiply(
    A=MaterialFunctionCall(Color_Morning=@n1.RGB, Color_Day=@n2.RGB, Color_Evening=@n1.RGB, Color_Night=@n3.RGB) {materialFunction=&Color_Time_of_Day}.Result
    B=Multiply(A=@n7.RGB, B=@n8.RGB)
   )) {materialFunction=&MF_FX_EmissiveAdjustments}
  n9 = TextureObjectParameter() {
   group=FrostTexture
   parameterName=FrostTexture_Texture
   samplerType=SAMPLERTYPE_Masks
   sortPriority=0
   texture=/Game/Global/Textures/Default/Noise/T_Metal_Noise
  }
  n10 = VectorParameter() {
   group=FrostTexture
   parameterName=FrostTexture_UVControl
   channelNames {a=VPan, b=UPan, g=VScale, r=UScale}
   defaultValue=<rgba 1.7991 0.5016 0 0>
  }
  n11 = ComponentMask(None=NamedRerouteUsage()) {g=true, r=true}
  n12 = OneMinus(None=MaterialFunctionCall(Texture=@n9, UVControl_RGB=@n10, UVControl_A=@n10, "[Override]UVInput"=@n11) {materialFunction=&FXCommon_TextureSample}.R)
  n13 = VectorParameter() {parameterName=Refine, defaultValue=<rgba 3.5 1 0.5 0.15>}
  n14 = Power(Base=@n12, Exp=@n13.R)
  n15 = Add(A=DynamicParameter() {parameterIndex=3, paramNames [Dissolve, "", "", ""]}.Dissolve, B=@n13.G)
  n16 = LinearInterpolate(
   A=MaterialFunctionCall(In=Multiply(A=@n14, B=@n15), Contrast=@n13.B) {materialFunction=&CheapContrast}.Result
   B=Multiply(A=@n12, B=@n13.A)
  )
  n17 = TextureSample(UVs=@n11) {texture=&/LanceMask}
  n18 = Saturate(None=Subtract(A=SceneTexture() {sceneTextureId=PPI_CustomDepth}.Color) {constB=10240})
  n19 = Constant() {r=1}
  n20 = Multiply(
   A=Multiply(A=ParticleColor().A, B=Multiply(A=@n16, B=@n17.R))
   B=MaterialFunctionCall("Xb1,PS4"=@n18, "PC-High&&Epic,XSX,PS5"=@n18, "PC-Perf,Android"=@n19, Switch=@n19, "PC-Low(NotPerf)&&Medium"=@n18) {materialFunction=&PlatformSwitch_1D}.Result
  )
  n21 = Power(Base=DepthFade(FadeDistance=ScalarParameter() {defaultValue=20, parameterName=DF}))
  n22 = MaterialFunctionCall(Float3=NamedRerouteUsage()) {materialFunction=&BreakOutFloat3Components}
  n23 = VectorParameter() {
   group=ZGradient
   parameterName=ZGradient_EdgeMask_InvertControl
   channelNames {a="", b=Power, g=Multiply, r=Offset}
   defaultValue=<rgba 0 7 3 1>
  }
  n24 = VectorParameter() {group=ZGradient, parameterName=ZGradient_EdgeMask_Control, channelNames {a="", b=Power, g=Multiply, r=Offset}, defaultValue=<rgba 0 7 3 1>}
  n25 = Normalize(VectorInput=DDY(Value=@n5.Result))
  n26 = Power(
   Base=Abs(None=MaterialFunctionCall(Float3=@n25) {materialFunction=&BreakOutFloat3Components}.B)
   Exp=Constant() {r=4}
  ) {constExponent=8}
  Multiply(
   A=OneMinus(None=Saturate(None=@n26))
   B=Multiply(
    A=Multiply(A=@n20, B=@n21)
    B=MaterialFunctionCall!(UorV=@n22.B, Invert_EdgeControl=@n23, NoInvert_EdgeControl=@n24) {materialFunction=/Game/Effects/Niagara/MasterMaterials/Functions/Common/FXCommon_EdgeMask_Ex}.Mask
   )
  )
  n27 = ParticlePositionWS()
  n28 = DynamicParameter() {paramNames [AxisAngleX,AxisAngleY,AxisAngleZ,AxisAngleW]}
  n29 = MaterialFunctionCall(SceneDepth=SceneDepth().R) {materialFunction=&WorldPositionBehindTranslucency}
  n30 = RotateAboutAxis(
   NormalizedRotationAxis=MaterialFunctionCall(X=@n28.AxisAngleX, Y=@n28.AxisAngleX, Z=@n28.AxisAngleX) {materialFunction=&MakeFloat3}.Result
   RotationAngle=@n28.AxisAngleW
   PivotPoint=@n27
   Position=@n29.Result
  )
  n31 = DynamicParameter() {parameterIndex=1, defaultValue=<rgba 1 1 1 0>, paramNames [BoxMask_X,BoxMask_Y,BoxMask_Z,Width]}
  n32 = Add(A=MaterialFunctionCall(X=@n31.BoxMask_X, Y=@n31.BoxMask_X, Z=@n31.BoxMask_X) {materialFunction=&MakeFloat3}.Result) {constB=0.1}
  n33 = Subtract(A=Subtract(A=@n27, B=Add(A=@n30, B=@n29.Result)), B=Multiply(A=@n32) {constB=0.5})
  n34 = Saturate(None=Ceil(None=Add(A=@n33, B=@n32)))
  n35 = Multiply(A=Min(A=Multiply(A=@n33, B=@n34)) {constB=0}) {constB=-1}
  n36 = MaterialFunctionCall(Float3=Ceil(None=Saturate(None=@n35))) {materialFunction=&BreakOutFloat3Components}
  n37 = Multiply(A=@n36.R, B=@n36.R)
  NamedRerouteDeclaration(None=Multiply(A=Multiply(A=@n37, B=@n36.B), B=Divide(A=@n35, B=@n32))) {name=BoxGradient, nodeColor=<rgba 0.4062 0 1 1>}
  ScalarParameter() {defaultValue=3, parameterName=TE}
 ]
}
