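# Material function for a dissolving ice decal, outputting color, opacity and a Fresnel-weighted sky-light reflection.
# UVs come from world position behind translucency, projected on XY, and a depth-derivative angle mask is derived.
# It builds main ice, switchable parallax-occlusion depth, camera-offset inner ice and masked, time-of-day-tinted snow layers.
# The Dissolve input thresholds noise for opacity, with a second, width-offset threshold blending an edge into the color.
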MaterialFunction /MotoWrestleGameplay/Assets/VFX/Materials/MF_IceDecal
graph {
 properties {previewBlendMode=BLEND_Translucent}
 nodes [
  n1 = Constant() {r=2}
  n2 = Add(A=NamedRerouteUsage())
  n3 = Constant() {r=2}
  n4 = OneMinus(None=FunctionInput() {bUsePreviewValueAsDefault=true, inputName=Dissolve, inputType=FunctionInput_Scalar})
  n5 = Multiply(A=@n4, B=Add(A=@n3, B=ScalarParameter() {defaultValue=-0.084, parameterName=DissolveEdgeWIdth}))
  n6 = Multiply(A=Multiply(A=@n1, B=Subtract(A=@n2, B=@n5)), B=ScalarParameter() {defaultValue=1, parameterName=Dissolve_MulBeforePower})
  n7 = Add(A=NamedRerouteUsage(), B=NamedRerouteUsage())
  n8 = NamedRerouteUsage()
  n9 = LinearInterpolate(
   A=LinearInterpolate(A=NamedRerouteUsage(), B=Add(A=@n7, B=@n8) {constB=0}, Alpha=Saturate(None=@n6)) {constA=1}
   B=NamedRerouteUsage()
   Alpha=NamedRerouteUsage()
  ) {constAlpha=0}
  FunctionOutput(None=NamedRerouteDeclaration(None=Divide(A=@n9)) {name=MainTextureOutput, nodeColor=<rgba 0.875 0 1 1>}) {outputName=Color}
  n10 = Multiply(A=@n3, B=@n4)
  n11 = Multiply(A=Multiply(A=Subtract(A=@n2, B=@n10), B=@n1), B=ScalarParameter() {defaultValue=1, parameterName=Dissolve_EdgeHardness})
  n12 = ScalarParameter() {defaultValue=1, parameterName=Dissolve_EdgeHardnessPower}
  n13 = MaterialFunctionCall(In=Saturate(None=Power(Base=@n11, Exp=@n12)), Contrast=ScalarParameter() {defaultValue=0.5, parameterName=DissolveContrast}) {materialFunction=&CheapContrast}
  n14 = NamedRerouteUsage()
  FunctionOutput(None=NamedRerouteDeclaration(None=Multiply(A=@n13.Result, B=@n14)) {name=OpacityOutput, nodeColor=<rgba 0 1 0.5547 1>}) {outputName=Opacity}
  n15 = VectorParameter() {parameterName=Reflection_FresnelAttribute, defaultValue=<rgba 12 0.01 7 0.1>}
  n16 = MaterialFunctionCall(Normal=NamedRerouteUsage(), Flatness=ScalarParameter() {defaultValue=1, parameterName=ReflectionNormalFlat}) {materialFunction=&FlattenNormal}
  n17 = Normalize(VectorInput=Transform(None=@n16.Result) {transformSourceType=TRANSFORMSOURCE_Instance})
  n18 = ReflectionVectorWS(CustomWorldNormal=Normalize(VectorInput=Transform(None=@n16.Result) {transformSourceType=TRANSFORMSOURCE_Instance}))
  n19 = Constant()
  FunctionOutput(None=Multiply(A=Fresnel(ExponentIn=@n15.R, BaseReflectFractionIn=@n15.R, Normal=@n17), B=SkyLightEnvMapSample(Direction=@n18, Roughness=@n19))) {outputName=Reflection}
  n20 = MaterialFunctionCall(SceneDepth=SceneDepth().R) {materialFunction=&WorldPositionBehindTranslucency}
  n21 = Normalize(VectorInput=DDY(Value=@n20.Result))
  n22 = Power(
   Base=Abs(None=MaterialFunctionCall(Float3=@n21) {materialFunction=&BreakOutFloat3Components}.B)
   Exp=Constant() {r=4}
  ) {constExponent=8}
  NamedRerouteDeclaration(None=OneMinus(None=Saturate(None=@n22))) {name=AngleMask, nodeColor=<rgba 0.1953 1 0 1>}
  n23 = VectorParameter() {
   group=SnowLayerMask
   parameterName=SnowLayerMask_UVControl
   channelNames {a=VPan, b=UPan, g=VScale, r=UScale}
   defaultValue=<rgba 0.6 0.6 0 0>
  }
  n24 = MaterialFunctionCall(
   Texture=TextureObject() {samplerType=SAMPLERTYPE_LinearGrayscale, texture=/Game/Creative/Sets/PropSets/Primitives/Customization/Textures/Ice/T_IceMask_02_M}
   UVControl_RGB=@n23
   UVControl_A=@n23
   "[Override]UVInput"=NamedRerouteUsage()
  ) {materialFunction=&FXCommon_TextureSample}
  n25 = ScalarParameter() {group=SnowLayerMask, parameterName=SnowLayerMask_GrayAdjust_Add}
  n26 = Saturate(None=Multiply(A=Add(A=@n24.R, B=@n25), B=ScalarParameter() {defaultValue=3, group=SnowLayerMask, parameterName=SnowLayerMask_GrayAdjust_Mul}))
  n27 = ScalarParameter() {defaultValue=1.5, group=SnowLayerMask, parameterName=SnowLayerMask_GrayAdjust_Power}
  NamedRerouteDeclaration(None=Saturate(None=Power(Base=@n26, Exp=@n27))) {name=SnowLayerMask, nodeColor=<rgba 0.8047 0 1 1>}
  n28 = VectorParameter() {group=SnowLayer, parameterName=SnowLayer_UVControl, channelNames {a=VPan, b=UPan, g=VScale, r=UScale}, defaultValue=<rgba 1 1 0 0>}
  n29 = Desaturation(
   None=MaterialFunctionCall(
    Texture=TextureObject() {texture=/Game/Packages/Fortnite_UniqueMaterials_1/Textures/T_Ice_Lake_Snow_D}
    UVControl_RGB=@n28
    UVControl_A=@n28
    "[Override]UVInput"=NamedRerouteUsage()
   ) {materialFunction=&FXCommon_TextureSample}.RGB
   Fraction=Constant() {r=1}
  )
  n30 = VectorParameter() {parameterName=C, defaultValue=<rgba 3 3 3 1>}
  n31 = VectorParameter() {parameterName=Color_Morning, defaultValue=<rgba 1 0.8676 0.8 1>}
  n32 = VectorParameter() {parameterName=Color_Day, defaultValue=<rgba 1 1 1 1>}
  n33 = VectorParameter() {parameterName=Color_Night, defaultValue=<rgba 0.65 0.7498 1 1>}
  NamedRerouteDeclaration(None=Multiply(
    A=Multiply(A=@n29, B=@n30.RGB)
    B=MaterialFunctionCall(Color_Morning=@n31.RGB, Color_Day=@n32.RGB, Color_Evening=@n31.RGB, Color_Night=@n33.RGB) {materialFunction=&Color_Time_of_Day}.Result
   )) {name=SnowTexture, nodeColor=<rgba 1 0.2344 0 1>}
  n34 = TextureObjectParameter() {
   group=MaskM
   parameterName=MaskM_Texture
   samplerType=SAMPLERTYPE_Masks
   sortPriority=0
   texture=/Game/Global/Textures/Default/Noise/T_Metal_Noise
  }
  n35 = VectorParameter() {group=MaskM, parameterName=MaskM_UVControl, channelNames {a=VPan, b=UPan, g=VScale, r=UScale}, defaultValue=<rgba 0.7 0.7 0 0>}
  n36 = MaterialFunctionCall(Texture=@n34, UVControl_RGB=@n35, UVControl_A=@n35, "[Override]UVInput"=NamedRerouteUsage()) {materialFunction=&FXCommon_TextureSample}
  NamedRerouteDeclaration(None=Power(Base=OneMinus(None=@n36.R), Exp=ScalarParameter() {defaultValue=2, parameterName=gh})) {name=DissolveTexture, nodeColor=<rgba 0 0.0859 1 1>}
  n37 = ComponentMask(None=FunctionInput(Preview=@n20.Result) {bUsePreviewValueAsDefault=true, inputName=WorldPosition}) {g=true, r=true}
  NamedRerouteDeclaration(None=Rotator(Coordinate=Multiply(A=@n37) {constB=0.003}, Time=Constant() {r=0.25})) {name=DecalWorldPositionUV, nodeColor=<rgba 0 1 0.2969 1>}
  n38 = VectorParameter() {
   group=MainIceTexture
   parameterName=MainIceTexture_UVControl
   channelNames {a=VPan, b=UPan, g=VScale, r=UScale}
   defaultValue=<rgba 0.5 0.5 0 0>
  }
  NamedRerouteDeclaration(None=MaterialFunctionCall(UVControl_RGB=@n38, UVControl_A=@n38, "[Override]UVInput"=NamedRerouteUsage()) {materialFunction=&FXCommon_TextureSample}.UV) {name=MainIceUV, nodeColor=<rgba 0.9922 1 0 1>}
  n39 = Constant()
  n40 = StaticBool() {value=true}
  n41 = MaterialFunctionCall!(
   CameraVec
   Heightmap Texture=TextureObject() {texture=&/IceD}
   Height Ratio=Constant() {r=-0.1}
   Min Steps=Constant() {r=32}
   Max Steps=Constant() {r=64}
   UVs=NamedRerouteUsage()
   Heightmap Channel=Constant4Vector() {constant=<rgba 0 0 1 0>}.RGBA
   Reference Plane=Constant() {r=-0.2}
   -----------------------------
   Use World Coordinates=@n40
   Specify Manual Texture Size=@n40
   Manual Texture Size
   ------------------------------
   "Render Shadows (Occlusion Mapping)"
   Light Vector
   Shadow Steps
   Shadow Penumbra
   Transform To VertexNormal=StaticBool()
  ) {materialFunction=&/ParallaxOcclusionMapping}
  NamedRerouteDeclaration(None=StaticSwitchParameter(True=@n39, False=@n41.Parallax UVs) {parameterName=DisablePrlx}) {name=ParallaxUV, nodeColor=<rgba 0.6875 0 1 1>}
  NamedRerouteDeclaration(None=StaticSwitchParameter(True=@n39, False=@n41.Pixel Depth Offset) {parameterName=DisablePrlx}) {name=ParallaxPixelDepthOffset, nodeColor=<rgba 0 1 0.3672 1>}
  NamedRerouteDeclaration(None=TextureSample(UVs=NamedRerouteUsage()) {samplerType=SAMPLERTYPE_Normal, texture=&/IceN}.RGB) {name=IceNormal, nodeColor=<rgba 1 0.0469 0 1>}
  n42 = CameraVectorWS()
  n43 = Add(A=Multiply(A=NamedRerouteUsage(), B=Constant() {r=1.25}), B=Multiply(A=ComponentMask(None=@n42) {g=true, r=true}, B=Constant() {r=-0.7616}))
  n44 = NamedRerouteUsage()
  n45 = Constant() {r=1.5}
  n46 = Multiply(
   A=ComponentMask(None=MaterialFunctionCall(Normal=@n44, Flatness=@n45) {materialFunction=&FlattenNormal}.Result) {g=true, r=true}
   B=Constant() {r=0.3}
  )
  n47 = Desaturation(
   None=TextureSample(UVs=Add(A=@n43, B=@n46), Tex=TextureObjectParameter() {parameterName=BumpTex, texture=&IceD}).RGB
   Fraction=ScalarParameter() {parameterName=BumpOffsetDesat}
  )
  n48 = VectorParameter() {parameterName=BumpOffsetTextureColorAndPower, defaultValue=<rgba 0.5115 0.7528 1 0.6557>}
  NamedRerouteDeclaration(None=Multiply(A=Power(Base=@n47, Exp=@n48.A), B=@n48.RGB)) {name=BumpOffsetTexture, nodeColor=<rgba 0 0.2734 1 1>}
  n49 = TextureObjectParameter() {group=MainIceTexture, parameterName=MainIceTexture_Texture, sortPriority=0, texture=&/Ice_D}
  n50 = Desaturation(
   None=TextureSample(UVs=NamedRerouteUsage(), Tex=@n49).RGB
   Fraction=ScalarParameter() {defaultValue=0.4, parameterName=MainIceTexture_Desaturation}
  )
  n51 = Constant() {r=1}
  n52 = VectorParameter() {parameterName=MainIceTexture_Refine, defaultValue=<rgba 1 1 1 1>}
  n53 = Multiply(
   A=Normalize(VectorInput=@n50)
   B=MaterialFunctionCall(Texture=Desaturation(None=@n50, Fraction=@n51), RGB=@n52.RGB, A=@n52.RGB) {materialFunction=&MF_FX_Refine}.Result
  )
  n54 = ScalarParameter() {parameterName=MainIceTexture_UseRefine}
  NamedRerouteDeclaration(None=StaticSwitchParameter(True=Constant(), False=LinearInterpolate(A=@n50, B=@n53, Alpha=@n54)) {parameterName=DisableMainIce}) {name=MainIce, nodeColor=<rgba 0.5938 1 0 1>}
  n55 = VectorParameter() {parameterName=Parallax_PixelDepthOffsetToColor, defaultValue=<rgba 0 230 0 1>}
  n56 = MaterialFunctionCall(Input=NamedRerouteUsage(), Input Low=@n55.R, Input High=@n55.R, Target Low=@n55.R, Target High=@n55.R) {materialFunction=&RemapValueRange}
  n57 = VectorParameter() {parameterName=ParallaxTextureColorAndPower, defaultValue=<rgba 0.325 0.5125 0.7 2>}
  n58 = Multiply(
   A=TextureSample(UVs=NamedRerouteUsage(), Tex=TextureObject() {texture=&IceD}).RGB
   B=Power(Base=Saturate(None=@n56.Result), Exp=@n57.A) {constExponent=5.2743}
  )
  NamedRerouteDeclaration(None=StaticSwitchParameter(True=Constant(), False=Multiply(A=@n58, B=@n57.RGB)) {parameterName=DisablePrlx}) {name=ParallaxTexture, nodeColor=<rgba 1 0 0.9141 1>}
 ]
}
