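# Translucent unlit MotoWrestle ice decal material for Niagara mesh particles, faking projection onto scene geometry.
# Depth-reconstructed world position and dynamic parameters form a rotatable box gradient, which supplies UVs for a mask texture.
# That mask plus dissolve offsets drives MF_IceDecal, whose color is particle-tinted before reflection is added.
# Opacity combines particle alpha, depth fade, a Z edge mask and custom-depth masking, skipped on Switch, Android and performance PC.
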Material /MotoWrestleGameplay/Assets/VFX/Materials/M_MotoWrestle_IceDecal
graph {
 properties {
  bAutomaticallySetUsageInEditor=false
  bUsedWithNiagaraMeshParticles=true
  blendMode=BLEND_Translucent
  shadingModel=MSM_Unlit
  twoSided=true
  usedShadingModels=Unlit
 }
 nodes [
  n1 = TextureObjectParameter() {
   parameterName=Mask
   samplerType=SAMPLERTYPE_LinearColor
   texture=/Game/Effects/Fort_Effects/Gameplay/SmokeDustExplosions/Textures/T_GroundChar_01
  }
  n2 = VectorParameter() {group=DecalShape, parameterName=DecalShape_UVControl, channelNames {a=VPan, b=UPan, g=VScale, r=UScale}, defaultValue=<rgba 1 1 0 0>}
  n3 = ParticlePositionWS()
  n4 = DynamicParameter() {defaultValue=<rgba 0 0 1 0>, paramNames [AxisAngleX,AxisAngleY,AxisAngleZ,AxisAngleW]}
  n5 = MaterialFunctionCall(SceneDepth=SceneDepth().R) {materialFunction=&WorldPositionBehindTranslucency}
  n6 = RotateAboutAxis(
   NormalizedRotationAxis=MaterialFunctionCall(X=@n4.AxisAngleX, Y=@n4.AxisAngleX, Z=@n4.AxisAngleX) {materialFunction=&MakeFloat3}.Result
   RotationAngle=@n4.AxisAngleW
   PivotPoint=@n3
   Position=@n5.Result
  )
  n7 = DynamicParameter() {parameterIndex=1, defaultValue=<rgba 1000 1000 1000 0>, paramNames [BoxMask_X,BoxMask_Y,BoxMask_Z,Param4]}
  n8 = Add(A=MaterialFunctionCall(X=@n7.BoxMask_X, Y=@n7.BoxMask_X, Z=@n7.BoxMask_X) {materialFunction=&MakeFloat3}.Result) {constB=0.1}
  n9 = Subtract(A=Subtract(A=@n3, B=Add(A=@n6, B=@n5.Result)), B=Multiply(A=@n8) {constB=0.5})
  n10 = Saturate(None=Ceil(None=Add(A=@n9, B=@n8)))
  n11 = Multiply(A=Min(A=Multiply(A=@n9, B=@n10)) {constB=0}) {constB=-1}
  n12 = MaterialFunctionCall(Float3=Ceil(None=Saturate(None=@n11))) {materialFunction=&BreakOutFloat3Components}
  n13 = Multiply(A=Multiply(A=Multiply(A=@n12.R, B=@n12.R), B=@n12.B), B=Divide(A=@n11, B=@n8))
  n14 = MaterialFunctionCall(
   Texture=@n1
   UVControl_RGB=@n2
   UVControl_A=@n2
   "[Override]UVInput"=ComponentMask(None=NamedRerouteDeclaration(None=@n13) {name=BoxGradient, nodeColor=<rgba 0.4062 0 1 1>}) {g=true, r=true}
  ) {materialFunction=&FXCommon_TextureSample}
  n15 = Add(A=DynamicParameter() {parameterIndex=2, defaultValue=<rgba 0 1 1 1>}.Param1, B=ScalarParameter() {parameterName=DDD})
  n16 = MaterialFunctionCall!(WorldPosition, Dissolve=Power(Base=Add(A=@n14.R, B=@n15), Exp=ScalarParameter() {defaultValue=1, parameterName=egb})) {materialFunction=&/MF_IceDecal}
  n17 = ParticleColor()
  Add(
   A=MaterialFunctionCall(Emissive=Multiply(A=@n16.Color, B=@n17.RGB)) {materialFunction=&MF_FX_EmissiveAdjustments}.Result
   B=@n16.Reflection
  )
  n18 = SceneTexture() {sceneTextureId=PPI_CustomDepth}
  Floor(None=Divide(A=@n18.Color) {constB=10240})
  n19 = Multiply(A=@n17.A, B=@n16.Opacity)
  n20 = Saturate(None=Subtract(A=@n18.Color) {constB=10240})
  n21 = Constant() {r=1}
  n22 = MaterialFunctionCall("Xb1,PS4"=@n20, "PC-High&&Epic,XSX,PS5"=@n20, "PC-Perf,Android"=@n21, Switch=@n21, "PC-Low(NotPerf)&&Medium"=@n20) {materialFunction=&PlatformSwitch_1D}
  n23 = DepthFade(FadeDistance=ScalarParameter() {defaultValue=20, parameterName=DF})
  n24 = NamedRerouteUsage()
  n25 = VectorParameter() {
   group=ZGradient
   parameterName=ZGradient_EdgeMask_InvertControl
   channelNames {a="", b=Power, g=Multiply, r=Offset}
   defaultValue=<rgba 0 2 3 1>
  }
  n26 = VectorParameter() {group=ZGradient, parameterName=ZGradient_EdgeMask_Control, channelNames {a="", b=Power, g=Multiply, r=Offset}, defaultValue=<rgba 0 20 2 1>}
  Multiply(
   A=Multiply(A=Multiply(A=@n19, B=@n22.Result), B=Power(Base=@n23))
   B=MaterialFunctionCall!(
    UorV=MaterialFunctionCall(Float3=@n24) {materialFunction=&BreakOutFloat3Components}.B
    Invert_EdgeControl=@n25
    NoInvert_EdgeControl=@n26
   ) {materialFunction=/Game/Effects/Niagara/MasterMaterials/Functions/Common/FXCommon_EdgeMask_Ex}.Mask
  )
  n27 = NamedRerouteUsage()
  n28 = VectorParameter() {
   group=ZGradient
   parameterName=ZGradient_EdgeMask_InvertControl
   channelNames {a="", b=Power, g=Multiply, r=Offset}
   defaultValue=<rgba 0 2 3 1>
  }
  n29 = VectorParameter() {group=ZGradient, parameterName=ZGradient_EdgeMask_Control, channelNames {a="", b=Power, g=Multiply, r=Offset}, defaultValue=<rgba 0 20 2 1>}
  Saturate(None=MaterialFunctionCall(
    UorV=MaterialFunctionCall(Float3=@n27) {materialFunction=&BreakOutFloat3Components}.B
    Invert_EdgeControl=@n28
    NoInvert_EdgeControl=@n29
   ) {materialFunction=&FXCommon_EdgeMask_Ex}.Mask)
  ScalarParameter() {defaultValue=0.5256}
 ]
}
