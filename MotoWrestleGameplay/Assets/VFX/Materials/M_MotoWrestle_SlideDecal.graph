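# Unlit translucent Niagara ribbon material that fakes a decal, by its name for MotoWrestle slide marks.
# With depth test disabled, it rebuilds world position behind the ribbon from scene depth and uses XY to map a texture tinted by particle color.
# The scene point's sideways offset from the particle, normalized by the RibbonWidth dynamic parameter, drives an edge mask.
# Opacity combines that mask, particle alpha and a pixel-versus-scene depth fade.
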Material /MotoWrestleGameplay/Assets/VFX/Materials/M_MotoWrestle_SlideDecal
graph {
 properties {
  bAutomaticallySetUsageInEditor=false
  bDisableDepthTest=true
  bUsedWithNiagaraRibbons=true
  blendMode=BLEND_Translucent
  shadingModel=MSM_Unlit
  usedShadingModels=Unlit
 }
 nodes [
  Constant2Vector() {g=1, r=1}
  MaterialFunctionCall(Float2=TextureCoordinate()) {materialFunction=&BreakOutFloat2Components}
  n1 = TextureObjectParameter() {
   group=MainTexture
   parameterName=MainTexture_Texture
   samplerType=SAMPLERTYPE_LinearColor
   sortPriority=0
   texture=/Engine/Functions/Engine_MaterialFunctions02/ExampleContent/Textures/OffsetNoiseDistanceFields_ForNormals
  }
  n2 = VectorParameter() {group=MainTexture, parameterName=MainTexture_UVControl, channelNames {a=VPan, b=UPan, g=VScale, r=UScale}, defaultValue=<rgba 1 1 0 0>}
  n3 = MaterialFunctionCall(SceneDepth=SceneDepth().R) {materialFunction=&WorldPositionBehindTranslucency}
  n4 = ComponentMask(None=@n3.Result) {g=true, r=true}
  n5 = ParticleColor()
  Multiply(
   A=MaterialFunctionCall(Texture=@n1, UVControl_RGB=@n2, UVControl_A=@n2, "[Override]UVInput"=Multiply(A=@n4) {constB=0.001}) {materialFunction=&FXCommon_TextureSample}.RGB
   B=@n5.RGB
  )
  n6 = ParticlePositionWS()
  n7 = DotProduct(
   A=MaterialFunctionCall!(Plane Normal=Constant3Vector() {constant=<rgba 0 0 1 0>}, Vector=Subtract(A=@n3.Result, B=@n6)) {materialFunction=/Engine/Functions/Engine_MaterialFunctions02/Math/ProjectVectorOntoPlane}.Result
   B=Transform(None=Constant3Vector() {constant=<rgba 0 1 0 0>})
  )
  n8 = DynamicParameter() {paramNames [RibbonWidth,Param2,Param3,Param4]}
  n9 = Divide(A=Add(A=Divide(A=@n7, B=@n8.RibbonWidth)))
  n10 = Multiply(
   A=@n9
   B=MaterialFunctionCall!(TexCoord=@n9, ClampChannel) {materialFunction=/Game/Effects/Fort_Effects/MaterialFunctions/MF_FX_TilingMethod}.Result
  )
  n11 = VectorParameter() {group=SideFade, parameterName=SideFade_EdgeMask_Control, channelNames {a="", b=Power, g=Multiply, r=Offset}, defaultValue=<rgba 0 4 1 1>}
  n12 = Subtract(A=PixelDepth().R, B=SceneDepth().R)
  n13 = ScalarParameter() {defaultValue=100, parameterName=T}
  Multiply(
   A=Multiply(
    A=@n5.A
    B=MaterialFunctionCall!(UorV=@n10, EdgeControl=@n11) {materialFunction=/Game/Effects/Niagara/MasterMaterials/Functions/Common/FXCommon_EdgeMask_Simple}.Mask
   )
   B=OneMinus(None=Divide(A=@n12, B=@n13))
  )
 ]
}
