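# Translucent, unlit, two-sided trail material for Niagara mesh particles in MotoWrestle, with index-of-refraction settings and SplineThicken width control.
# A dynamic Offset parameter shifts a gradient scan, which Perlin noise distorts with side-mask weighting.
# The refined scan, particle alpha and side and tip/tail fades drive a blend toward the Ref value.
# Plasma texture, oversaturated through negative desaturation and tinted by particle color, feeds emissive.
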Material /MotoWrestleGameplay/Assets/VFX/Materials/M_MotoWrestle_Trail
graph {
 properties {
  bAutomaticallySetUsageInEditor=false
  bUsedWithNiagaraMeshParticles=true
  blendMode=BLEND_Translucent
  refractionMethod=RM_IndexOfRefraction
  shadingModel=MSM_Unlit
  twoSided=true
  usedShadingModels=Unlit
 }
 nodes [
  n1 = ParticleColor()
  n2 = TextureObjectParameter() {group=ScanT, parameterName=ScanT_Texture, sortPriority=0, texture=/Game/Global/Textures/Default/Masks_and_Gradients/T_Vert_Gradient_Soft}
  n3 = VectorParameter() {parameterName=Offset, defaultValue=<rgba 0 -0.0425 0 1>}
  n4 = Constant()
  n5 = DynamicParameter() {paramNames [Offset,Param2,Param3,Param4]}
  n6 = VectorParameter() {group=ScanT, parameterName=ScanT_UVControl, channelNames {a=VPan, b=UPan, g=VScale, r=UScale}, defaultValue=<rgba 1 1.197 0 0>}
  n7 = MaterialFunctionCall(Float2=TextureCoordinate()) {materialFunction=&BreakOutFloat2Components}
  n8 = MaterialFunctionCall(
   AdditionalUVOffset=Add(A=AppendVector(A=@n3.R, B=@n3.R), B=AppendVector(A=@n4, B=@n5.Offset))
   UVControl_RGB=@n6
   UVControl_A=@n6
   "[Override]UVInput"=MaterialFunctionCall(X=@n7.R, Y=@n7.R) {materialFunction=&MakeFloat2}.Result
  ) {materialFunction=&FXCommon_TextureSample}
  n9 = Saturate(None=MaterialFunctionCall(Float2=@n8.UV) {materialFunction=&BreakOutFloat2Components}.G)
  n10 = ScalarParameter() {defaultValue=1, parameterName=sssawf}
  n11 = MaterialFunctionCall(X=Constant(), Y=ScalarParameter() {defaultValue=-0.9204, parameterName=fw}) {materialFunction=&MakeFloat2}
  n12 = TextureObjectParameter() {
   group=NoiseA
   parameterName=NoiseA_Texture
   samplerType=SAMPLERTYPE_LinearGrayscale
   sortPriority=0
   texture=&T_SimplePerlin_Greyscale
  }
  n13 = VectorParameter() {group=NoiseA, parameterName=NoiseA_UVControl, channelNames {a=VPan, b=UPan, g=VScale, r=UScale}, defaultValue=<rgba 0.5 0.7248 0 1>}
  n14 = VectorParameter() {
   group=NoiseA
   parameterName=NoiseA_NoiseControl
   sortPriority=60
   channelNames {a=VStrength, b=UStrength, g=VOffset, r=UOffset}
   defaultValue=<rgba -0.5 -0.5 1.2769 2.8107>
  }
  n15 = MaterialFunctionCall!(
   NoiseValue=OneMinus(None=MaterialFunctionCall(Texture=@n12, UVControl_RGB=@n13, UVControl_A=@n13) {materialFunction=&FXCommon_TextureSample}.RG)
   Control_RGB=@n14
   Control_A=@n14
  ) {materialFunction=/Game/Effects/Niagara/MasterMaterials/Functions/Common/FXCommon_NoiseControl}
  n16 = MaterialFunctionCall(Float2=TextureCoordinate()) {materialFunction=&BreakOutFloat2Components}
  n17 = VectorParameter() {
   group=NoiseStrSideMask
   parameterName=NoiseStrSideMask_EdgeMask_Control
   channelNames {a="", b=Power, g=Multiply, r=Offset}
   defaultValue=<rgba 0 1 1 1>
  }
  n18 = MaterialFunctionCall!(UorV=@n16.R, EdgeControl=@n17) {materialFunction=/Game/Effects/Niagara/MasterMaterials/Functions/Common/FXCommon_EdgeMask_Simple}
  n19 = MaterialFunctionCall(
   Texture=@n2
   AdditionalUVOffset=Multiply(A=Power(Base=@n9, Exp=@n10), B=LinearInterpolate(A=@n11.Result, B=@n15.NoiseValue, Alpha=@n18.Mask))
   UVControl_RGB=@n6
   UVControl_A=@n6
   "[Override]UVInput"=@n8.UV
  ) {materialFunction=&FXCommon_TextureSample}
  n20 = VectorParameter() {parameterName=ScanTRefine, defaultValue=<rgba 13 10 0.3 0.5>}
  n21 = Constant2Vector() {g=1}
  n22 = Multiply(
   A=@n1.A
   B=Multiply(
    A=MaterialFunctionCall(Texture=@n19.R, RGB=@n20.RGB, A=@n20.RGB) {materialFunction=&MF_FX_Refine}.Result
    B=MaterialFunctionCall!(TexCoord=@n19.UV, ClampChannel=@n21) {materialFunction=/Game/Effects/Fort_Effects/MaterialFunctions/MF_FX_TilingMethod}.Result
   )
  )
  n23 = VectorParameter() {
   group=SideFadeMask
   parameterName=SideFadeMask_EdgeMask_Control
   channelNames {a="", b=Power, g=Multiply, r=Offset}
   defaultValue=<rgba 0 2 3 1>
  }
  n24 = MaterialFunctionCall(UorV=@n16.R, EdgeControl=@n23) {materialFunction=&FXCommon_EdgeMask_Simple}
  n25 = VectorParameter() {
   group=TipTailFade
   parameterName=TipTailFade_EdgeMask_InvertControl
   channelNames {a="", b=Power, g=Multiply, r=Offset}
   defaultValue=<rgba 0 1.5 2 1>
  }
  n26 = VectorParameter() {
   group=TipTailFade
   parameterName=TipTailFade_EdgeMask_Control
   channelNames {a="", b=Power, g=Multiply, r=Offset}
   defaultValue=<rgba 0 5 3 1>
  }
  LinearInterpolate(
   B=ScalarParameter() {defaultValue=-5, parameterName=Ref}
   Alpha=Multiply(
    A=Multiply(A=@n22, B=@n24.Mask)
    B=MaterialFunctionCall!(UorV=@n16.G, Invert_EdgeControl=@n25, NoInvert_EdgeControl=@n26) {materialFunction=/Game/Effects/Niagara/MasterMaterials/Functions/Common/FXCommon_EdgeMask_Ex}.Mask
   )
  ) {constA=1}
  n27 = ScalarParameter() {defaultValue=50, parameterName=W}
  MaterialFunctionCall(WidthBase=@n27, WidthTip=@n27) {materialFunction=&SplineThicken}
  n28 = TextureObjectParameter() {
   group=GroupA
   parameterName=GroupA_Texture
   samplerType=SAMPLERTYPE_LinearColor
   sortPriority=0
   texture=/Game/Effects/Fort_Effects/Emotes/UltraLight/Texture/T_Plasma_A_C
  }
  n29 = VectorParameter() {group=GroupA, parameterName=GroupA_UVControl, channelNames {a=VPan, b=UPan, g=VScale, r=UScale}, defaultValue=<rgba 0.3 0.5 0 0>}
  n30 = MaterialFunctionCall(Texture=@n28, UVControl_RGB=@n29, UVControl_A=@n29, "[Override]UVInput"=@n8.UV) {materialFunction=&FXCommon_TextureSample}
  n31 = Normalize(VectorInput=@n30.RGB)
  n32 = ScalarParameter() {defaultValue=-10, parameterName=DES}
  n33 = VectorParameter() {parameterName=GroupARef, defaultValue=<rgba 7 4.3347 0 0.5>}
  n34 = Multiply(
   A=Multiply(
    A=Desaturation(None=@n31, Fraction=@n32)
    B=MaterialFunctionCall(Texture=@n30.R, RGB=@n33.RGB, A=@n33.RGB) {materialFunction=&MF_FX_Refine}.Result
   )
   B=VectorParameter() {parameterName=C, defaultValue=<rgba 10 10 10 1>}.RGB
  )
  n35 = ScalarParameter() {defaultValue=0.5, parameterName=FEB}
  MaterialFunctionCall(Emissive=Multiply(A=Add(A=@n34, B=@n35), B=@n1.RGB)) {materialFunction=&MF_FX_EmissiveAdjustments}
  OneMinus(None=@n7.G)
 ]
}
