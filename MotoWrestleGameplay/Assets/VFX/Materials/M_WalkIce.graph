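# Translucent unlit Niagara sprite/mesh particle material for a walk-on ice effect, apparently over water, with depth test disabled.
# A particle-driven sphere mask dissolves MF_IceDecal, blended with a platform-scaled, dithered parallax ray-march of noise for depth.
# Four offset custom-depth and shading-model-ID taps, plus depth comparisons, mask where it appears.
# A BottomLayer switch swaps in vertex-colour tinting and fade with dithered edges.
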Material /MotoWrestleGameplay/Assets/VFX/Materials/M_WalkIce
graph {
 properties {
  bAutomaticallySetUsageInEditor=false
  bDisableDepthTest=true
  bUseMaterialAttributes=true
  bUsedWithNiagaraMeshParticles=true
  bUsedWithNiagaraSprites=true
  bUsedWithWater=true
  blendMode=BLEND_Translucent
  shadingModel=MSM_Unlit
  twoSided=true
  usedShadingModels=Unlit
 }
 nodes [
  n1 = SceneDepth()
  n2 = PixelDepth()
  n3 = ScalarParameter() {defaultValue=10, parameterName=DepthFade}
  Clamp(None=Divide(A=Subtract(A=@n1.R, B=@n2.R), B=@n3))
  Constant3Vector()
  Constant() {r=1}
  n4 = SphereMask(A=TextureCoordinate(), B=Constant() {r=0.5}) {attenuationRadius=1, hardnessPercent=0}
  n5 = VectorParameter() {group=EdgeFade, parameterName=EdgeFade_EdgeMask_Control, channelNames {a="", b=Power, g=Multiply, r=Offset}, defaultValue=<rgba 0 1.5 2 1>}
  n6 = Power(Base=Multiply(A=Add(A=@n4, B=@n5.Offset), B=@n5.Multiply), Exp=@n5.Power)
  n7 = DynamicParameter() {paramNames [Param1,AdditionalX,AdditionalY,Radius]}
  MaterialExpressionScalarParameter_8 = ScalarParameter() {parameterName=SphereMaskHardness}
  n8 = SphereMask(
   A=TextureCoordinate()
   B=ScalarParameter() {defaultValue=0.5, parameterName=SphereMaskSize}
   Radius=@n7.Param1
   Hardness=@MaterialExpressionScalarParameter_8
  )
  n9 = MaterialFunctionCall!(WorldPosition=WorldPosition().XYZ, Dissolve=Multiply(A=Saturate(None=@n6), B=@n8)) {materialFunction=&/MF_IceDecal}
  n10 = OneMinus(None=VertexColor().R)
  n11 = Power(Base=@n10, Exp=ScalarParameter() {defaultValue=1, parameterName=VCPower})
  n12 = ScalarParameter() {defaultValue=1, parameterName=VCMul}
  n13 = Subtract(A=SceneTexture() {sceneTextureId=PPI_CustomDepth}.Color, B=@n2.R)
  n14 = ScreenPosition()
  n15 = CameraPositionWS()
  n16 = WorldPosition()
  n17 = ViewProperty() {property=MEVP_TemporalSampleIndex}
  MaterialExpressionAdd_8 = Add(A=ScreenPosition().PixelPosition, B=AppendVector(A=@n17.Property, B=@n17.Property))
  n18 = Custom!(p=@MaterialExpressionAdd_8) {
   code="Mod( ((uint)(p.x) + 2 * (uint)(p.y)) , 5 )"
   outputType=CMOT_Float1
   inputs [{inputName=p, input {expression=&self:MaterialExpressionAdd_8}}]
  }
  n19 = Multiply(
   A=Divide(A=ScalarParameter() {defaultValue=20, parameterName=F}, B=Distance(A=@n15, B=@n16.XYZ))
   B=LinearInterpolate(
    A=ScalarParameter() {defaultValue=0.1, parameterName=FLT}
    Alpha=Multiply(A=@n18) {constB=0.1666, desc=Divide by 6 avoiding a precision problem where an Alpha Threshold of 0 still leaves some remaining pixels.}
   )
  )
  n20 = MaterialFunctionCall!(OffsetAmount=MaterialFunctionCall(X=@n19, Y=@n19) {materialFunction=&MakeFloat2}.Result) {materialFunction=/Engine/Functions/Engine_MaterialFunctions01/ImageAdjustment/BlurSampleOffsets}
  n21 = Add(A=@n14.ViewportUV, B=@n20."0,1")
  n22 = MaterialFunctionCall(Float3=SceneTexture(UVs=@n21) {sceneTextureId=PPI_ShadingModelID}.Color) {materialFunction=&BreakOutFloat3Components}
  n23 = SceneTexture(UVs=@n21) {sceneTextureId=PPI_CustomDepth}
  n24 = Ceil(None=Saturate(None=Subtract(A=@n23.Color) {constB=999999}))
  n25 = Constant() {r=1}
  n26 = Add(A=@n14.ViewportUV, B=@n20."1,0")
  n27 = MaterialFunctionCall(Float3=SceneTexture(UVs=@n26) {sceneTextureId=PPI_ShadingModelID}.Color) {materialFunction=&BreakOutFloat3Components}
  n28 = SceneTexture(UVs=@n26) {sceneTextureId=PPI_CustomDepth}
  n29 = Ceil(None=Saturate(None=Subtract(A=@n28.Color) {constB=999999}))
  n30 = Constant() {r=1}
  n31 = Add(
   A=If(A=LinearInterpolate(B=@n22.R, Alpha=@n24) {constA=10}, "A > B"=@n25, "A == B"=Constant(), "A < B"=@n25) {constB=10}
   B=If(A=LinearInterpolate(B=@n27.R, Alpha=@n29) {constA=10}, "A > B"=@n30, "A == B"=Constant(), "A < B"=@n30) {constB=10}
  )
  n32 = Add(A=@n14.ViewportUV, B=@n20."0,-1")
  n33 = SceneTexture(UVs=@n32) {sceneTextureId=PPI_ShadingModelID}
  n34 = Saturate(None=Subtract(A=SceneTexture(UVs=@n32) {sceneTextureId=PPI_CustomDepth}.Color) {constB=999999})
  n35 = Constant() {r=1}
  n36 = If(
   A=LinearInterpolate(
    B=MaterialFunctionCall(Float3=@n33.Color) {materialFunction=&BreakOutFloat3Components}.R
    Alpha=Ceil(None=@n34)
   ) {constA=10}
   "A > B"=@n35
   "A == B"=Constant()
   "A < B"=@n35
  ) {constB=10}
  n37 = Add(A=@n14.ViewportUV, B=@n20."-1,0")
  n38 = Subtract(A=SceneTexture(UVs=@n37) {sceneTextureId=PPI_CustomDepth}.Color) {constB=999999}
  n39 = LinearInterpolate(
   B=MaterialFunctionCall(Float3=SceneTexture(UVs=@n37) {sceneTextureId=PPI_ShadingModelID}.Color) {materialFunction=&BreakOutFloat3Components}.R
   Alpha=Ceil(None=Saturate(None=@n38))
  ) {constA=10}
  n40 = Constant() {r=1}
  n41 = Constant()
  n42 = Divide(A=Add(A=Add(A=@n31, B=@n36), B=If(A=@n39, "A > B"=@n40, "A == B"=@n41, "A < B"=@n40) {constB=10})) {constB=4}
  n43 = ScalarParameter() {defaultValue=1, parameterName=WaterIDPower}
  n44 = LinearInterpolate(A=Clamp(None=Divide(A=@n13, B=@n3)), B=Constant(), Alpha=Saturate(None=Power(Base=@n42, Exp=@n43))) {constA=1, constAlpha=1}
  n45 = Constant()
  MaterialExpressionClamp_0 = Clamp(None=@n7.Param1) {minDefault=0.0001}
  MaterialExpressionNormalize_0 = Normalize(VectorInput=CameraVectorWS())
  MaterialExpressionComponentMask_3 = ComponentMask(None=Transform(None=@MaterialExpressionNormalize_0) {transformSourceType=TRANSFORMSOURCE_World, transformType=TRANSFORM_Tangent}) {g=true, r=true}
  n46 = Constant() {r=4}
  MaterialExpressionMaterialFunctionCall_16 = MaterialFunctionCall("Xb1,PS4"=@n46, "PC-High&&Epic,XSX,PS5"=Constant() {r=7}, "PC-Perf,Android"=@n46, Switch=@n46, "PC-Low(NotPerf)&&Medium"=@n46) {materialFunction=&PlatformSwitch_1D}
  MaterialExpressionDivide_0 = Divide(A=ScalarParameter() {defaultValue=-0.1, parameterName=Test_Offset}, B=@MaterialExpressionMaterialFunctionCall_16.Result)
  MaterialExpressionTextureCoordinate_4 = TextureCoordinate()
  MaterialExpressionTextureObject_0 = TextureObject() {samplerType=SAMPLERTYPE_Masks, texture=/Game/Global/Textures/Default/Noise/T_Metal_Noise}
  n47 = ComponentMask(None=WorldPosition().XYZ) {g=true, r=true}
  MaterialExpressionRotator_0 = Rotator(Coordinate=Multiply(A=Multiply(A=@n47) {constB=0.003}) {constB=0.7}, Time=Constant() {r=0.25})
  MaterialExpressionScalarParameter_22 = ScalarParameter() {defaultValue=1.5, parameterName=Test_DissolvePower}
  MaterialExpressionScalarParameter_23 = ScalarParameter() {defaultValue=1, parameterName=Test_DissolveEdgeHardness}
  MaterialExpressionScalarParameter_25 = ScalarParameter() {defaultValue=1, parameterName=Test_DissolveEdgeHardnessPow}
  MaterialExpressionScalarParameter_24 = ScalarParameter() {defaultValue=0.5, parameterName=Test_DissolveContrast}
  n48 = Constant() {r=1}
  MaterialExpressionMaterialFunctionCall_17 = MaterialFunctionCall(
   Input=MaterialFunctionCall(Random=Constant()) {materialFunction=&DitherTemporalAA_Fortnite}.Result
   Input Low=Constant()
   Input High=@n48
   Target Low=ScalarParameter() {defaultValue=0.8, parameterName=Test_InputLow}
   Target High=@n48
  ) {materialFunction=&RemapValueRange}
  MaterialExpressionScalarParameter_28 = ScalarParameter() {defaultValue=0.5, parameterName=ZP}
  n49 = Custom!(
   Radius=@MaterialExpressionClamp_0
   Hardness=@MaterialExpressionScalarParameter_8
   CameraVectorTS=@MaterialExpressionComponentMask_3
   Offset=@MaterialExpressionDivide_0
   StepCount=@MaterialExpressionMaterialFunctionCall_16.Result
   UV=@MaterialExpressionTextureCoordinate_4
   DissolveTex=@MaterialExpressionTextureObject_0
   DissolveTexUV=@MaterialExpressionRotator_0
   CameraVectorWS=@MaterialExpressionNormalize_0
   DissolveValuePower=@MaterialExpressionScalarParameter_22
   DissolveEdgeHardness=@MaterialExpressionScalarParameter_23
   DissolveEdgeHardnessPow=@MaterialExpressionScalarParameter_25
   DissolveContrast=@MaterialExpressionScalarParameter_24
   Dither=@MaterialExpressionMaterialFunctionCall_17.Result
   ZPower=@MaterialExpressionScalarParameter_28
  ) {
   code="//Input\r\n//float Radius;\r\n//float Hardness;\r\n//float2 CameraVectorTS;\r\n//float Offset;\r\n//float StepCount;\r\n//float2 UV;\r\n//TexOBJ DissolveTex;\r\n//float2 DissolveTexUV;\r\n//float3 CameraVectorWS;\r\n//float DissolveValuePower;\r\n//float DissolveEdgeHardness;\r\n//float DissolveEdgeHardnessPow;\r\n//float DissolveContrast;\r\n//float Dither;\r\nint i = 0;\r\nfloat MaskSum = 0;\r\nwhile(i < StepCount){\r\nfloat2 OffsetedUV = UV+CameraVectorTS*(i+1)*Offset*Dither;\r\nfloat2 OffsetedDissolveUV = DissolveTexUV+CameraVectorWS*(i+1)*Offset*Dither;\r\nfloat SphereMask = 1-saturate(distance(OffsetedUV,0.5)/Radius);\r\nfloat3 DissolveValue = Texture2DSample(DissolveTex,DissolveTexSampler,OffsetedDissolveUV);\r\nfloat AdjustedDissolveValue = PositiveClampedPow((1-DissolveValue.r),DissolveValuePower);\r\nfloat DissolveResult = saturate(PositiveClampedPow((AdjustedDissolveValue+1-(2*(1-SphereMask)))*2*DissolveEdgeHardness,DissolveEdgeHardnessPow));\r\nfloat AdjustedDissolveResult = saturate(lerp(0-DissolveContrast,DissolveContrast+1,DissolveResult));\r\nMaskSum += AdjustedDissolveResult*PositiveClampedPow((1-(float)i/StepCount),ZPower);\r\ni++;\r\n}\r\nMaskSum /= StepCount;\r\nreturn MaskSum;"
   showCode=true
   inputs [
    {inputName=Radius, input {expression=&self:MaterialExpressionClamp_0}}
    {inputName=Hardness, input {expression=&self:MaterialExpressionScalarParameter_8}}
    {inputName=CameraVectorTS, input {expression=&self:MaterialExpressionComponentMask_3}}
    {inputName=Offset, input {expression=&self:MaterialExpressionDivide_0}}
    {inputName=StepCount, input {expression=&self:MaterialExpressionMaterialFunctionCall_16}}
    {inputName=UV, input {expression=&self:MaterialExpressionTextureCoordinate_4}}
    {inputName=DissolveTex, input {expression=&self:MaterialExpressionTextureObject_0}}
    {inputName=DissolveTexUV, input {expression=&self:MaterialExpressionRotator_0}}
    {inputName=CameraVectorWS, input {expression=&self:MaterialExpressionNormalize_0}}
    {inputName=DissolveValuePower, input {expression=&self:MaterialExpressionScalarParameter_22}}
    {inputName=DissolveEdgeHardness, input {expression=&self:MaterialExpressionScalarParameter_23}}
    {
     inputName=DissolveEdgeHardnessPow
     input {expression=&self:MaterialExpressionScalarParameter_25}
    }
    {inputName=DissolveContrast, input {expression=&self:MaterialExpressionScalarParameter_24}}
    {inputName=Dither, input {expression=&self:MaterialExpressionMaterialFunctionCall_17}}
    {inputName=ZPower, input {expression=&self:MaterialExpressionScalarParameter_28}}
   ]
  }
  n50 = Multiply(A=@n49, B=ScalarParameter() {defaultValue=1, parameterName=TEST_BlendOpacity})
  n51 = Saturate(None=Multiply(A=@n9.Opacity, B=ScalarParameter() {defaultValue=1.5, parameterName=Test_BlendM}))
  n52 = Multiply(
   A=StaticSwitchParameter(True=Multiply(A=@n11, B=@n12), False=Constant() {r=1}) {parameterName=BottomLayer}
   B=Multiply(
    A=MaterialFunctionCall("Xb1,PS4"=@n44, "PC-High&&Epic,XSX,PS5"=@n44, "PC-Perf,Android"=@n45, Switch=@n45, "PC-Low(NotPerf)&&Medium"=@n44) {materialFunction=&PlatformSwitch_1D}.Result
    B=LinearInterpolate(A=@n50, B=@n9.Opacity, Alpha=@n51)
   )
  )
  n53 = Constant()
  n54 = If(
   A=@n52
   "A > B"=MaterialFunctionCall(Alpha Threshold=@n52) {materialFunction=&DitherTemporalAA}.Result
   "A == B"=@n53
   "A < B"=@n53
  )
  n55 = Power(Base=@n52, Exp=ScalarParameter() {defaultValue=2, parameterName=DitherEdge})
  n56 = Multiply(A=StaticSwitchParameter(True=LinearInterpolate(A=@n54, Alpha=@n55), False=@n52) {parameterName=BottomLayer}, B=ParticleColor().A)
  n57 = Subtract(A=PixelDepth().R, B=SceneDepth().R)
  n58 = Constant() {r=300}
  n59 = Constant()
  n60 = Constant() {r=1}
  n61 = VectorParameter() {parameterName=BottomLayerColor, defaultValue=<rgba 1 1 1 1>}
  n62 = StaticSwitchParameter(
   True=LinearInterpolate(A=@n61.RGB, B=VectorParameter() {parameterName=BottomLayerColorB, defaultValue=<rgba 1 1 1 1>}.RGB, Alpha=@n10)
   False=LinearInterpolate(A=@n61.RGB, B=@n9.Color, Alpha=@n51)
  ) {parameterName=BottomLayer}
  MakeMaterialAttributes!(
   BaseColor
   Metallic
   Specular
   Roughness
   Anisotropy
   EmissiveColor=MaterialFunctionCall(Emissive=Divide(A=@n62)) {materialFunction=&MF_FX_EmissiveAdjustments}.Result
   Opacity=Multiply(A=Saturate(None=@n56), B=If(A=@n57, B=@n58, "A > B"=@n59, "A == B"=@n60, "A < B"=@n60))
   OpacityMask=MaterialFunctionCall(Alpha Threshold=@n9.Opacity) {materialFunction=&DitherTemporalAA}.Result
   Normal
   Tangent
   WorldPositionOffset
   SubsurfaceColor
   ClearCoat
   ClearCoatRoughness
   AmbientOcclusion
   Refraction
   CustomizedUVs_0
   CustomizedUVs_1
   CustomizedUVs_2
   CustomizedUVs_3
   CustomizedUVs_4
   CustomizedUVs_5
   CustomizedUVs_6
   CustomizedUVs_7
   PixelDepthOffset
   ShadingModel
   Displacement
  )
  MakeMaterialAttributes()
  n63 = ComponentMask(None=TransformPosition(None=WorldPosition().XYZ) {transformSourceType=TRANSFORMPOSSOURCE_World}) {g=true, r=true}
  n64 = DynamicParameter() {paramNames [Param1,AdditionalX,AdditionalY,Radius]}
  n65 = AppendVector(A=@n64.AdditionalX, B=@n64.AdditionalX)
  n66 = Multiply(A=@n7.Param1, B=@n64.Radius)
  n67 = DynamicParameter() {parameterIndex=1, paramNames [AdditionalAreaOpacity, Param2, Param3, Param4]}
  n68 = ScalarParameter() {defaultValue=0.9, parameterName=AdditionalAreaStrength}
  Max(
   A=@n8
   B=Multiply(A=SphereMask(A=@n63, B=@n65, Radius=@n66, Hardness=@MaterialExpressionScalarParameter_8), B=Multiply(A=@n67.AdditionalAreaOpacity, B=@n68))
  )
  Max(B=@n9.Opacity)
  ScalarParameter() {defaultValue=0.4, parameterName=Test_Radius}
  ScalarParameter() {defaultValue=1, parameterName=Test_Hardness}
 ]
}
