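# Niagara system for the MotoWrestle hit effect.
# Seven emitters make up the effect: spike streaks, star and flare flashes, sparks, smoke, ice pieces and a big glow.
# System scripts combine three variable-setting modules and SystemState with dynamic inputs for color multiplication, vector packing and a Fortnite time-of-day scalar.
# User parameters expose a delay, swing section index and front-end and preview flags.
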NiagaraSystem /MotoWrestleGameplay/Assets/VFX/NS_MotoWrestle_HitFX
dependencies {
 usedDataInterfaces []
 usedDynamicInputs [
  &MultiplyLinearColor
  &MakeFloat4FromVectorAndFloat
  &VectorFromFloat
  &FNTimeOfDayScalar
 ]
 usedModules [
  &self:SystemSpawnScript.SystemScriptSource.SystemScriptGraph.NiagaraNodeAssignment_0.SetVariables_#1
  &self:SystemSpawnScript.SystemScriptSource.SystemScriptGraph.NiagaraNodeAssignment_1.SetVariables_#2
  &self:SystemSpawnScript.SystemScriptSource.SystemScriptGraph.NiagaraNodeAssignment_2.SetVariables_#3
  &SystemState
 ]
}
emitters {
 BigGlow {rendererClasses [], renderers []}
 FlashFlare {rendererClasses [], renderers []}
 FlashStar {rendererClasses [], renderers []}
 IcePiece {rendererClasses [], renderers []}
 Smoke002 {rendererClasses [], renderers []}
 Sparks {rendererClasses [], renderers []}
 SpikeStreaks {rendererClasses [], renderers []}
}
summary {
 systemName=NS_MotoWrestle_HitFX
 emitters [
  {emitterName=SpikeStreaks, rendererClasses []}
  {emitterName=FlashStar, rendererClasses []}
  {emitterName=FlashFlare, rendererClasses []}
  {emitterName=Sparks, rendererClasses []}
  {emitterName=Smoke002, rendererClasses []}
  {emitterName=IcePiece, rendererClasses []}
  {emitterName=BigGlow, rendererClasses []}
 ]
 userVariables [
  User.bIsFrontEnd {type=NiagaraBool}
  User.bIsFrontEndPreview {type=NiagaraBool}
  User.Delay {type=NiagaraFloat, defaultValue=0, description=""}
  User.SwingSectionIndex {type=NiagaraInt32}
 ]
}
