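# NS_MotoWrestle_PickupFX is a Niagara system in MotoWrestleGameplay, named as a pickup effect.
# It holds two emitters, Smoke and Smoke001, with no renderers listed for either.
# SystemState is the only module listed, and a User.Delay float is exposed, though its use is not shown.
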NiagaraSystem /MotoWrestleGameplay/Assets/VFX/NS_MotoWrestle_PickupFX
dependencies {usedDataInterfaces [], usedDynamicInputs [], usedModules [&SystemState]}
emitters {Smoke {rendererClasses [], renderers []}, Smoke001 {rendererClasses [], renderers []}}
summary {
 systemName=NS_MotoWrestle_PickupFX
 emitters [{emitterName=Smoke001, rendererClasses []}, {emitterName=Smoke, rendererClasses []}]
 userVariables [User.Delay {type=NiagaraFloat, defaultValue=0, description=""}]
}
